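# Control Rig for the M_MED_FluteLamp character body that procedurally corrects baggy shirt and sleeve bones as limbs move.
# Spherical pose readers on thighs and upper arms, often curve-remapped, produce weights for additive bone offsets, alongside swing-twist DeltaRotation helpers.
# Shirt logic branches into crouch, DBNO and default setups, with parent constraints pulling cloth toward thighs in the first two.
# Interpolated state bools and flags like Emote, Pistol and OpenGlider scale specific corrections.
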ControlRigBlueprint /BRCosmetics/Characters/Player/Male/Medium/Bodies/M_MED_FluteLamp/Meshes/M_MED_FluteLamp_CtrlRig
graphs [
 Multiply_10_ContainedGraph {nodes [
   n1 = RigVMFunctionEntryNode()
   RigVMFunctionReturnNode(Result=&RigVMFunction_MathDoubleMul(B=@n1.C, A=&RigVMFunction_MathDoubleMul(A=@n1.A, B=@n1.B).Result).Result)
  ]}
 Multiply_6_ContainedGraph {nodes [
   n1 = RigVMFunctionEntryNode()
   RigVMFunctionReturnNode(Result=&RigVMFunction_MathDoubleMul(B=@n1.C, A=&RigVMFunction_MathDoubleMul(A=@n1.A, B=@n1.B).Result).Result)
  ]}
 Multiply_6_ContainedGraph {nodes [
   n1 = RigVMFunctionEntryNode()
   RigVMFunctionReturnNode(Result=&RigVMFunction_MathDoubleMul(B=@n1.C, A=&RigVMFunction_MathDoubleMul(A=@n1.A, B=@n1.B).Result).Result)
  ]}
 New Function_ContainedGraph {nodes [
   DeltaRotation_1 = RigVMFunctionReferenceNode(
    ExecuteContext=()
    ParentType=Null
    child=thigh_l
    debugRotation=False
    max=10
    min=5
    onlyTwist=False
    parent=thigh_l_null
    primAxis=1
    rotationOrder=XYZ
    twistAxis=(X=1,Y=0,Z=0)
   )
   > &RigUnit_ModifyTransforms(
    ItemToModify="((Item=(Type=Bone,Name=\"base_fr_ShirtBagginess_c\"),Transform=(Rotation=(X=0,Y=0,Z=0.1373,W=0.9905),Translation=(X=0,Y=0,Z=0),Scale3D=(X=1,Y=1,Z=1))),(Item=(Type=Bone,Name=\"base_fr_ShirtBagginess_l\"),Transform=(Rotation=(X=0,Y=0,Z=0,W=-1),Translation=(X=0,Y ...(truncated)"
    WeightMinimum=0
    Weight=@DeltaRotation_1.ouputWeight
   )
   RigVMFunctionEntryNode()
   > &RigUnit_ParentConstraint(
    AdvancedSettings=(InterpolationType=Average,RotationOrderForFilter=XZY)
    Child=(Type=Bone,Name="base_ArmBagginessArea_l")
    Filter=(TranslationFilter=(bX=False,bY=False,bZ=False),RotationFilter=(bX=True,bY=True,bZ=True),ScaleFilter=(bX=False,bY=False,bZ=False))
    Parents=((Item=(Type=Bone,Name="upperarm_twist_02_l"),Weight=1))
    Weight=1
   )
   > &RigUnit_ParentConstraint(
    AdvancedSettings=(InterpolationType=Average,RotationOrderForFilter=XZY)
    Child=(Type=Bone,Name="base_ArmBagginessArea_r")
    Filter=(TranslationFilter=(bX=False,bY=False,bZ=False),RotationFilter=(bX=True,bY=True,bZ=True),ScaleFilter=(bX=False,bY=False,bZ=False))
    Parents=((Item=(Type=Bone,Name="upperarm_twist_02_r"),Weight=1))
    Weight=1
   )
   > SPRs Leg and InterpolateVariables = RigVMFunctionReferenceNode()
   RigVMFunction_ControlFlowBranch = &RigVMFunction_ControlFlowBranch(ExecuteContext=@"SPRs Leg and InterpolateVariables".ExecuteContext, Condition=RigVMVariableNode() {variable=CrouchingBool}.Value)
   RigVMFunctionReturnNode(ExecuteContext=@RigVMFunction_ControlFlowBranch.Completed)
   n1 = &RigVMFunction_Sequence(ExecuteContext=@RigVMFunction_ControlFlowBranch.True)
   &RigUnit_ParentConstraint(
    AdvancedSettings=(InterpolationType=Average,RotationOrderForFilter=XZY)
    Child=(Type=Bone,Name="base_fr_ShirtBagginess_l")
    Parents=((Item=(Type=Bone,Name="thigh_l"),Weight=1))
    Weight=0.9
    ExecutePin=@n1.A
   )
   > &RigUnit_ParentConstraint(
    AdvancedSettings=(InterpolationType=Average,RotationOrderForFilter=XZY)
    Child=(Type=Bone,Name="base_fr_ShirtBagginess_r")
    Parents=((Item=(Type=Bone,Name="thigh_r"),Weight=1))
    Weight=0.9
   )
   > &RigUnit_ParentConstraint(
    AdvancedSettings=(InterpolationType=Average,RotationOrderForFilter=XZY)
    Child=(Type=Bone,Name="base_sd_ShirtBagginess_l")
    Parents=((Item=(Type=Bone,Name="thigh_l"),Weight=1))
    Weight=0.7
   )
   > &RigUnit_ParentConstraint(
    AdvancedSettings=(InterpolationType=Average,RotationOrderForFilter=XZY)
    Child=(Type=Bone,Name="base_sd_ShirtBagginess_r")
    Parents=((Item=(Type=Bone,Name="thigh_r"),Weight=1))
    Weight=0.7
   )
   &RigUnit_ModifyTransforms(
    ItemToModify="((Item=(Type=Bone,Name=\"base_bk_ShirtBagginess_l\"),Transform=(Rotation=(X=0.0718,Y=-0.0383,Z=-0.0905,W=-0.9926),Translation=(X=0.3503,Y=-1.0138,Z=1.2229),Scale3D=(X=1,Y=1,Z=1))),(Item=(Type=Bone,Name=\"base_bk_ShirtBagginess_c\"),Transform=(Rotation=(X=0,Y=0 ...(truncated)"
    WeightMinimum=0
    ExecutePin=@n1.B
    Weight=RigVMVariableNode() {variable=Crouching}.Value
   )
   n2 = RigVMVariableNode() {variable=Leg_L_Front_Crouch}
   > &RigUnit_ModifyTransforms(
    ItemToModify="((Item=(Type=Bone,Name=\"base_fr_ShirtBagginess_l\"),Transform=(Rotation=(X=-0.1997,Y=0.0337,Z=-0.1699,W=0.9644),Translation=(X=22.9518,Y=-0.0849,Z=2.1709),Scale3D=(X=0.4182,Y=1,Z=1))),(Item=(Type=Bone,Name=\"dyn_fr_ShirtBagginess_l_02\"),Transform=(Rotation=( ...(truncated)"
    WeightMinimum=0
    Weight=@n2.Value
   )
   > DeltaRotation_3 = RigVMFunctionReferenceNode(
    ParentType=Null
    child=thigh_l
    debugRotation=False
    max=-90
    min=0
    onlyTwist=False
    parent=thigh_l_null
    primAxis=1
    rotationOrder=XYZ
    twistAxis=(X=1,Y=0,Z=0)
   )
   > &RigUnit_ModifyTransforms(
    ItemToModify="((Item=(Type=Bone,Name=\"base_fr_ShirtBagginess_l\"),Transform=(Rotation=(X=0,Y=0,Z=0,W=1),Translation=(X=11.9239,Y=-3.264,Z=-0.0545),Scale3D=(X=0.5368,Y=1,Z=1))),(Item=(Type=Bone,Name=\"base_sd_ShirtBagginess_l\"),Transform=(Rotation=(X=0.033,Y=-0.1437,Z=-0.1 ...(truncated)"
    WeightMinimum=0
    Weight=RigVMVariableNode() {variable=Leg_L_Side}.Value
   )
   n3 = RigVMVariableNode() {variable=Leg_R_Front_Crouch}
   > &RigUnit_ModifyTransforms(
    ItemToModify="((Item=(Type=Bone,Name=\"dyn_fr_ShirtBagginess_c_02\"),Transform=(Rotation=(X=-0.338,Y=0.4376,Z=0.0448,W=-0.832),Translation=(X=-2.9795,Y=0,Z=-2.8359),Scale3D=(X=1,Y=1,Z=1))),(Item=(Type=Bone,Name=\"base_sd_ShirtBagginess_r\"),Transform=(Rotation=(X=-0.0175,Y= ...(truncated)"
    WeightMinimum=0
    Weight=@n3.Value
   )
   > &RigUnit_ModifyTransforms(
    ItemToModify="((Item=(Type=Bone,Name=\"base_fr_ShirtBagginess_r\"),Transform=(Rotation=(X=0,Y=0,Z=0.0513,W=0.9987),Translation=(X=-6.2943,Y=4.7263,Z=0.0465),Scale3D=(X=0.9426,Y=1,Z=1))),(Item=(Type=Bone,Name=\"base_fr_ShirtBagginess_c\"),Transform=(Rotation=(X=0,Y=0,Z=0.022 ...(truncated)"
    WeightMinimum=0
    Weight=RigVMVariableNode() {variable=Leg_R_Side}.Value
   )
   Multiply_9 = RigVMAggregateNode(C=1, A=RigVMVariableNode() {variable=Leg_L_Front_Crouch}.Value, B=RigVMVariableNode() {variable=Leg_R_Front_Crouch}.Value)
   > &RigUnit_ModifyTransforms(
    ItemToModify="((Item=(Type=Bone,Name=\"dyn_fr_ShirtBagginess_c_01\"),Transform=(Rotation=(X=0.2685,Y=-0.0962,Z=0.2329,W=-0.9297),Translation=(X=0,Y=0,Z=0),Scale3D=(X=1,Y=1,Z=1))),(Item=(Type=Bone,Name=\"dyn_fr_ShirtBagginess_c_02\"),Transform=(Rotation=(X=-0.0186,Y=0,Z=0,W= ...(truncated)"
    WeightMinimum=0
    Weight=@Multiply_9.Result
   )
   Multiply_5 = RigVMAggregateNode(B=1, C=RigVMVariableNode() {variable=Moving2D}.Value, A=RigVMVariableNode() {variable=Leg_L_Front_Crouch}.Value)
   > &RigUnit_ModifyTransforms(
    ItemToModify=((Item=(Type=Bone,Name="base_fr_ShirtBagginess_c"),Transform=(Rotation=(X=0,Y=-0.2164,Z=0,W=0.9763),Translation=(X=0,Y=0,Z=0),Scale3D=(X=1,Y=1,Z=1))))
    WeightMinimum=0
    Weight=@Multiply_5.Result
   )
   n4 = &RigVMFunction_ControlFlowBranch(Condition=RigVMVariableNode() {variable=DBNO}.Value, ExecuteContext=@RigVMFunction_ControlFlowBranch.False)
   RigVMFunction_Sequence = RigVMAggregateNode(ExecuteContext=@n4.False)
   &RigUnit_ModifyTransforms(
    ItemToModify=((Item=(Type=Bone,Name="base_fr_ShirtBagginess_r"),Transform=(Rotation=(X=-0.0009,Y=0.0331,Z=0.028,W=-0.9991),Translation=(X=0,Y=0,Z=0),Scale3D=(X=1,Y=1,Z=1))))
    Weight=1
    WeightMinimum=0
    ExecutePin=@RigVMFunction_Sequence.B
   )
   n5 = RigVMVariableNode() {variable=Leg_R_Front}
   > &RigUnit_ModifyTransforms(
    ItemToModify="((Item=(Type=Bone,Name=\"base_fr_ShirtBagginess_r\"),Transform=(Rotation=(X=-0.0385,Y=0.199,Z=0.0992,W=-0.9742),Translation=(X=0.1879,Y=0.0254,Z=0.248),Scale3D=(X=1,Y=1,Z=1))),(Item=(Type=Bone,Name=\"base_sd_ShirtBagginess_r\"),Transform=(Rotation=(X=0.0047,Y= ...(truncated)"
    WeightMinimum=0
    Weight=@n5.Value
   )
   > DeltaRotation = RigVMFunctionReferenceNode(
    ParentType=Null
    child=thigh_r
    debugRotation=False
    max=-110
    min=-90
    onlyTwist=False
    parent=thigh_r_null
    primAxis=2
    rotationOrder=XYZ
    twistAxis=(X=1,Y=0,Z=0)
   )
   n6 = &RigVMFunction_AnimEvalRichCurve(
    Curve="(EditorCurveData=(Keys=((InterpMode=RCIM_Cubic),(InterpMode=RCIM_Cubic,TangentMode=RCTM_User,Time=0.5977,Value=0.386,ArriveTangent=1.1374,LeaveTangent=1.1374),(InterpMode=RCIM_Cubic,TangentMode=RCTM_User,Time=1,Value=1,ArriveTangent=1.044,LeaveTangent=1.04 ...(truncated)"
    SourceMinimum=0
    TargetMinimum=0
    Value=@DeltaRotation.ouputWeight
   )
   > &RigUnit_ModifyTransforms(
    ItemToModify="((Item=(Type=Bone,Name=\"base_fr_ShirtBagginess_r\"),Transform=(Rotation=(X=-0.0584,Y=0.0195,Z=0.0012,W=0.9981),Translation=(X=3.7827,Y=0.0001,Z=1.2941),Scale3D=(X=1,Y=1,Z=1))),(Item=(Type=Bone,Name=\"base_sd_ShirtBagginess_r\"),Transform=(Rotation=(X=0,Y=0,Z= ...(truncated)"
    WeightMinimum=0
    Weight=@n6.Result
   )
   > &RigUnit_ModifyTransforms(
    ItemToModify="((Item=(Type=Bone,Name=\"base_fr_ShirtBagginess_r\"),Transform=(Rotation=(X=0.0087,Y=0.0126,Z=-0.3233,W=0.9462),Translation=(X=0.7056,Y=-0.3103,Z=-0.0124),Scale3D=(X=0.7,Y=1,Z=1))),(Item=(Type=Bone,Name=\"base_sd_ShirtBagginess_r\"),Transform=(Rotation=(X=-0.0 ...(truncated)"
    WeightMinimum=0
    Weight=RigVMVariableNode() {variable=Leg_R_Side}.Value
   )
   > &RigUnit_ModifyTransforms(
    ItemToModify="((Item=(Type=Bone,Name=\"base_sd_ShirtBagginess_r\"),Transform=(Rotation=(X=0,Y=0,Z=0.042,W=-0.9991),Translation=(X=0,Y=0,Z=0),Scale3D=(X=1,Y=1,Z=1))),(Item=(Type=Bone,Name=\"base_bk_ShirtBagginess_r\"),Transform=(Rotation=(X=-0.2199,Y=0.4412,Z=0.2021,W=-0.846 ...(truncated)"
    WeightMinimum=0
    Weight=RigVMVariableNode() {variable=Leg_R_Back}.Value
   )
   > &RigUnit_ModifyTransforms(
    ItemToModify="((Item=(Type=Bone,Name=\"dyn_fr_ShirtBagginess_r_02\"),Transform=(Rotation=(X=-0.022,Y=-0.3536,Z=-0.2494,W=0.9013),Translation=(X=-2.1258,Y=0.9442,Z=2.2903),Scale3D=(X=1,Y=1,Z=1))),(Item=(Type=Bone,Name=\"dyn_fr_ShirtBagginess_c_02\"),Transform=(Rotation=(X=0, ...(truncated)"
    WeightMinimum=0
    Weight=RigVMVariableNode() {variable=Leg_R_Front_Mix}.Value
   )
   n7 = &RigVMFunction_AnimEvalRichCurve(
    Curve=(EditorCurveData=(Keys=((),(InterpMode=RCIM_Cubic,TangentMode=RCTM_User,Time=0.543,Value=0.4154,ArriveTangent=0.9648,LeaveTangent=0.9648),(InterpMode=RCIM_Cubic,TangentMode=RCTM_User,Time=1,Value=1,ArriveTangent=1.2832,LeaveTangent=1.2832))))
    SourceMinimum=0
    TargetMinimum=0
    Value=RigVMVariableNode() {variable=Leg_L_Front}.Value
   )
   &RigUnit_ModifyTransforms(
    ItemToModify="((Item=(Type=Bone,Name=\"dyn_fr_ShirtBagginess_l_02\"),Transform=(Rotation=(X=-0.0294,Y=-0.1538,Z=-0.0387,W=0.9869),Translation=(X=-2.4546,Y=-1.8447,Z=-2.9952),Scale3D=(X=0.5,Y=1.1292,Z=1))),(Item=(Type=Bone,Name=\"base_sd_ShirtBagginess_l\"),Transform=(Rotati ...(truncated)"
    WeightMinimum=0
    Weight=@n7.Result
    ExecutePin=@RigVMFunction_Sequence.A
   )
   > DeltaRotation_5 = RigVMFunctionReferenceNode(
    ParentType=Null
    child=thigh_l
    debugRotation=False
    max=-110
    min=-90
    onlyTwist=False
    parent=thigh_l_null
    primAxis=2
    rotationOrder=XYZ
    twistAxis=(X=1,Y=0,Z=0)
   )
   > &RigUnit_ModifyTransforms(
    ItemToModify="((Item=(Type=Bone,Name=\"base_fr_ShirtBagginess_l\"),Transform=(Rotation=(X=-0.0067,Y=-0.203,Z=0.0324,W=0.9786),Translation=(X=0,Y=0,Z=0),Scale3D=(X=1,Y=1,Z=1))),(Item=(Type=Bone,Name=\"base_fr_ShirtBagginess_c\"),Transform=(Rotation=(X=-0.0985,Y=0.0036,Z=0.03 ...(truncated)"
    Weight=0
    WeightMinimum=0
   )
   > &RigUnit_ModifyTransforms(
    ItemToModify="((Item=(Type=Bone,Name=\"base_fr_ShirtBagginess_l\"),Transform=(Rotation=(X=0,Y=-0.0641,Z=0,W=0.9979),Translation=(X=0,Y=0,Z=0),Scale3D=(X=1,Y=1,Z=1))),(Item=(Type=Bone,Name=\"base_sd_ShirtBagginess_l\"),Transform=(Rotation=(X=0.0197,Y=-0.7934,Z=0.0327,W=0.607 ...(truncated)"
    WeightMinimum=0
    Weight=RigVMVariableNode() {variable=Leg_L_Side}.Value
   )
   > &RigUnit_ModifyTransforms(
    ItemToModify="((Item=(Type=Bone,Name=\"base_sd_ShirtBagginess_l\"),Transform=(Rotation=(X=0.0045,Y=-0.0564,Z=-0.0801,W=0.9952),Translation=(X=0,Y=0,Z=0),Scale3D=(X=1,Y=1,Z=1))),(Item=(Type=Bone,Name=\"base_bk_ShirtBagginess_l\"),Transform=(Rotation=(X=0.1334,Y=-0.5072,Z=-0. ...(truncated)"
    WeightMinimum=0
    Weight=RigVMVariableNode() {variable=Leg_L_Back}.Value
   )
   > &RigUnit_ModifyTransforms(
    ItemToModify="((Item=(Type=Bone,Name=\"dyn_fr_ShirtBagginess_l_02\"),Transform=(Rotation=(X=-0.1232,Y=-0.3726,Z=-0.5748,W=0.718),Translation=(X=0,Y=0,Z=0),Scale3D=(X=1,Y=1,Z=1))),(Item=(Type=Bone,Name=\"dyn_fr_ShirtBagginess_c_02\"),Transform=(Rotation=(X=0.0064,Y=-0.0669,Z ...(truncated)"
    WeightMinimum=0
    Weight=RigVMVariableNode() {variable=Leg_L_Front_Mix}.Value
   )
   Remap = &RigVMFunction_MathDoubleRemap(
    SourceMaximum=0.3
    SourceMinimum=0
    TargetMaximum=1
    TargetMinimum=0
    Value=&RigVMFunction_MathDoubleMul(A=RigVMVariableNode() {variable=Leg_L_Front}.Value, B=RigVMVariableNode() {variable=Leg_R_Front}.Value).Result
   )
   &RigUnit_ModifyTransforms(
    ItemToModify="((Item=(Type=Bone,Name=\"base_fr_ShirtBagginess_c\"),Transform=(Rotation=(X=0.0088,Y=-0.0243,Z=-0.1713,W=0.9849),Translation=(X=0.79,Y=-0.1483,Z=4.8614),Scale3D=(X=1,Y=1,Z=1))),(Item=(Type=Bone,Name=\"base_fr_ShirtBagginess_l\"),Transform=(Rotation=(X=-0.0025, ...(truncated)"
    WeightMinimum=0
    Weight=@Remap.Result
    ExecutePin=@RigVMFunction_Sequence.C
   )
   > n8 = &RigUnit_SphericalPoseReader(
    ActiveRegionScaleFactors=(PositiveWidth=1,NegativeWidth=1,PositiveHeight=1,NegativeHeight=1)
    ActiveRegionSize=0.1
    Debug=(bDrawDebug=False,bDraw2D=False,bDrawLocalAxes=False,DebugScale=25,DebugSegments=20,DebugThickness=0.25)
    DriverAxis=(X=-1,Y=0,Z=0)
    DriverItem=(Type=Bone,Name="thigh_l")
    FalloffRegionScaleFactors=(PositiveWidth=2,NegativeWidth=2,PositiveHeight=3,NegativeHeight=3)
    FalloffSize=0.2
    FlipHeightScaling=False
    FlipWidthScaling=False
    OptionalParentItem=(Type=Bone,Name="None")
    RotationOffset=(X=110,Y=0,Z=-50)
   )
   n9 = &RigVMFunction_MathDoubleMul(A=@n8.OutputParam, B=RigVMVariableNode() {variable=Emote}.Value)
   > &RigUnit_ModifyTransforms(
    ItemToModify="((Item=(Type=Bone,Name=\"dyn_fr_ShirtBagginess_l_02\"),Transform=(Rotation=(X=-0.0006,Y=-0.1829,Z=0.0111,W=0.9831),Translation=(X=-12.1224,Y=-0.0422,Z=-3.9011),Scale3D=(X=1,Y=1,Z=1))),(Item=(Type=Bone,Name=\"base_sd_ShirtBagginess_l\"),Transform=(Rotation=(X=0 ...(truncated)"
    WeightMinimum=0
    Weight=@n9.Result
   )
   > n10 = &RigUnit_SphericalPoseReader(
    ActiveRegionScaleFactors=(PositiveWidth=1,NegativeWidth=0.3,PositiveHeight=1.5,NegativeHeight=1.5)
    ActiveRegionSize=0.1
    Debug=(bDrawDebug=False,bDraw2D=False,bDrawLocalAxes=False,DebugScale=25,DebugSegments=20,DebugThickness=0.25)
    DriverAxis=(X=1,Y=0,Z=0)
    DriverItem=(Type=Bone,Name="thigh_r")
    FalloffRegionScaleFactors=(PositiveWidth=1,NegativeWidth=2,PositiveHeight=4,NegativeHeight=4)
    FalloffSize=0.2
    FlipHeightScaling=false
    FlipWidthScaling=true
    OptionalParentItem=(Type=Bone,Name="None")
    RotationOffset=(X=-90,Y=0,Z=-45)
   )
   n11 = &RigVMFunction_MathDoubleMul(A=@n10.OutputParam, B=RigVMVariableNode() {variable=Emote}.Value)
   > &RigUnit_ModifyTransforms(
    ItemToModify="((Item=(Type=Bone,Name=\"base_fr_ShirtBagginess_r\"),Transform=(Rotation=(X=0,Y=-0.2395,Z=0,W=0.9709),Translation=(X=4.5317,Y=0,Z=1.1448),Scale3D=(X=1,Y=1,Z=1))),(Item=(Type=Bone,Name=\"base_sd_ShirtBagginess_r\"),Transform=(Rotation=(X=0,Y=0,Z=0,W=1),Translat ...(truncated)"
    WeightMinimum=0
    Weight=@n11.Result
   )
   &RigUnit_ParentConstraint(
    AdvancedSettings=(InterpolationType=Average,RotationOrderForFilter=XZY)
    Child=(Type=Bone,Name="base_fr_ShirtBagginess_l")
    Parents=((Item=(Type=Bone,Name="thigh_l"),Weight=1))
    Weight=0.9
    ExecutePin=@n4.True
   )
   > &RigUnit_ParentConstraint(
    AdvancedSettings=(InterpolationType=Average,RotationOrderForFilter=XZY)
    Child=(Type=Bone,Name="base_fr_ShirtBagginess_r")
    Parents=((Item=(Type=Bone,Name="thigh_r"),Weight=1))
    Weight=0.9
   )
   > &RigUnit_ParentConstraint(
    AdvancedSettings=(InterpolationType=Average,RotationOrderForFilter=XZY)
    Child=(Type=Bone,Name="base_sd_ShirtBagginess_l")
    Parents=((Item=(Type=Bone,Name="thigh_l"),Weight=1))
    Weight=0.5
   )
   > &RigUnit_ParentConstraint(
    AdvancedSettings=(InterpolationType=Average,RotationOrderForFilter=XZY)
    Child=(Type=Bone,Name="base_sd_ShirtBagginess_r")
    Parents=((Item=(Type=Bone,Name="thigh_r"),Weight=1))
    Weight=0.5
   )
   > &RigUnit_ModifyTransforms(
    ItemToModify="((Item=(Type=Bone,Name=\"base_fr_ShirtBagginess_l\"),Transform=(Rotation=(X=0.0065,Y=0.0345,Z=-0.0773,W=0.9964),Translation=(X=20.616,Y=1.8355,Z=5.0574),Scale3D=(X=0.3342,Y=2.1392,Z=1))),(Item=(Type=Bone,Name=\"base_sd_ShirtBagginess_l\"),Transform=(Rotation=( ...(truncated)"
    WeightMinimum=0
    Weight=RigVMVariableNode() {variable=Leg_L_Front}.Value
   )
   > &RigUnit_ModifyTransforms(
    ItemToModify="((Item=(Type=Bone,Name=\"base_bk_ShirtBagginess_r\"),Transform=(Rotation=(X=-0.0411,Y=0.0955,Z=-0.0273,W=0.9942),Translation=(X=0.8995,Y=-0.7036,Z=-0.4865),Scale3D=(X=1,Y=1,Z=1))),(Item=(Type=Bone,Name=\"base_sd_ShirtBagginess_r\"),Transform=(Rotation=(X=0.017 ...(truncated)"
    WeightMinimum=0
    Weight=RigVMVariableNode() {variable=Leg_R_Front}.Value
   )
   > &RigUnit_ParentConstraint(
    AdvancedSettings=(InterpolationType=Average,RotationOrderForFilter=XZY)
    Child=(Type=Bone,Name="base_bk_ShirtBagginess_c")
    Parents=((Item=(Type=Bone,Name="base_bk_ShirtBagginess_l"),Weight=1),(Item=(Type=Bone,Name="base_bk_ShirtBagginess_r"),Weight=1))
    Weight=0.5
   )
   > DeltaRotation_2 = RigVMFunctionReferenceNode(
    ParentType=Null
    child=thigh_r
    debugRotation=False
    max=-120
    min=-100
    onlyTwist=False
    parent=thigh_r_null
    primAxis=2
    rotationOrder=XYZ
    twistAxis=(X=1,Y=0,Z=0)
   )
   > &RigUnit_ModifyTransforms(
    ItemToModify="((Item=(Type=Bone,Name=\"base_fr_ShirtBagginess_r\"),Transform=(Rotation=(X=0,Y=-0.0324,Z=0,W=0.9995),Translation=(X=0.2165,Y=-0.8523,Z=-1.1882),Scale3D=(X=1,Y=1.4647,Z=1))),(Item=(Type=Bone,Name=\"base_fr_ShirtBagginess_c\"),Transform=(Rotation=(X=0.0562,Y=-0 ...(truncated)"
    WeightMinimum=0
    Weight=@DeltaRotation_2.ouputWeight
   )
   DeltaRotation_4 = RigVMFunctionReferenceNode(
    ExecuteContext=()
    ParentType=Null
    child=thigh_r
    debugRotation=False
    max=-90
    min=0
    onlyTwist=False
    parent=thigh_r_null
    primAxis=1
    rotationOrder=XYZ
    twistAxis=(X=1,Y=0,Z=0)
   )
   &RigUnit_ModifyTransforms(
    ExecutePin=()
    ItemToModify="((Item=(Type=Bone,Name=\"base_fr_ShirtBagginess_c\"),Transform=(Rotation=(X=0,Y=-0.2164,Z=0,W=0.9763),Translation=(X=0,Y=0,Z=0),Scale3D=(X=1,Y=1,Z=1))),(Item=(Type=Bone,Name=\"base_sd_ShirtBagginess_r\"),Transform=(Rotation=(X=-0.046,Y=-0.254,Z=0.181,W=0.949), ...(truncated)"
    Weight=0
    WeightMinimum=0
   )
   &RigUnit_ModifyTransforms(
    ExecutePin=()
    ItemToModify="((Item=(Type=Bone,Name=\"dyn_fr_ShirtBagginess_r_02\"),Transform=(Rotation=(X=-0.1103,Y=-0.3275,Z=-0.2829,W=0.8948),Translation=(X=0,Y=0,Z=0),Scale3D=(X=1,Y=1,Z=1))),(Item=(Type=Bone,Name=\"dyn_fr_ShirtBagginess_c_02\"),Transform=(Rotation=(X=0,Y=0,Z=0,W=1),Tr ...(truncated)"
    Weight=1
    WeightMinimum=0
   )
   &RigUnit_ModifyTransforms(
    ExecutePin=()
    ItemToModify="((Item=(Type=Bone,Name=\"base_fr_ShirtBagginess_c\"),Transform=(Rotation=(X=-0.0023,Y=0.0258,Z=-0.0878,W=-0.9958),Translation=(X=3.1769,Y=0,Z=4.9826),Scale3D=(X=1,Y=1,Z=1))),(Item=(Type=Bone,Name=\"base_fr_ShirtBagginess_l\"),Transform=(Rotation=(X=-0.0025,Y=0 ...(truncated)"
    Weight=1
    WeightMinimum=0
   )
   &RigUnit_ModifyTransforms(
    ExecutePin=()
    ItemToModify="((Item=(Type=Bone,Name=\"base_fr_ShirtBagginess_r\"),Transform=(Rotation=(X=0.0168,Y=0.0043,Z=-0.193,W=0.981),Translation=(X=-0.3571,Y=0.1461,Z=0.0058),Scale3D=(X=0.7,Y=1,Z=1))),(Item=(Type=Bone,Name=\"base_sd_ShirtBagginess_r\"),Transform=(Rotation=(X=-0.0827 ...(truncated)"
    Weight=1
    WeightMinimum=0
   )
   Multiply_6 = RigVMAggregateNode(
    B=RigVMVariableNode() {variable=Leg_R_Front_Crouch}.Value
    C=RigVMVariableNode() {variable=Moving2D}.Value
    A=RigVMVariableNode() {variable=Arm_L_Up}.Value
   )
   &RigUnit_ModifyTransforms(
    ExecutePin=()
    ItemToModify=((Item=(Type=Bone,Name="None"),Transform=(Rotation=(X=0,Y=0,Z=0,W=1),Translation=(X=0,Y=0,Z=0),Scale3D=(X=1,Y=1,Z=1))))
    WeightMinimum=0
    Weight=@Multiply_6.Result
   )
   &RigUnit_ModifyTransforms(
    ExecutePin=()
    ItemToModify="((Item=(Type=Bone,Name=\"dyn_fr_ShirtBagginess_c_01\"),Transform=(Rotation=(X=0.2685,Y=-0.0962,Z=0.2329,W=-0.9297),Translation=(X=0,Y=0,Z=0),Scale3D=(X=1,Y=1,Z=1))),(Item=(Type=Bone,Name=\"dyn_fr_ShirtBagginess_c_02\"),Transform=(Rotation=(X=-0.0186,Y=0,Z=0,W= ...(truncated)"
    Weight=1
    WeightMinimum=0
   )
   &RigUnit_ModifyTransforms(
    ExecutePin=()
    ItemToModify="((Item=(Type=Bone,Name=\"dyn_fr_ShirtBagginess_c_02\"),Transform=(Rotation=(X=-0.0767,Y=0.405,Z=-0.1578,W=0.8973),Translation=(X=4.1628,Y=0.9672,Z=2.0445),Scale3D=(X=1,Y=1,Z=1))),(Item=(Type=Bone,Name=\"dyn_fr_ShirtBagginess_r_02\"),Transform=(Rotation=(X=0.12 ...(truncated)"
    Weight=0
    WeightMinimum=0
   )
   &RigVMFunction_MathDoubleMul(B=1, A=@n2.Value)
   &RigVMFunction_MathDoubleMul(B=1, A=@n3.Value)
   &RigVMFunction_MathDoubleRemap(SourceMaximum=1, SourceMinimum=0, TargetMaximum=0, TargetMinimum=1, bClamp=false, Value=RigVMVariableNode() {variable=Moving2D}.Value)
   RigVMFunction_AnimEvalRichCurve = &RigVMFunction_AnimEvalRichCurve(
    Curve=(EditorCurveData=(Keys=((InterpMode=RCIM_Cubic),(InterpMode=RCIM_Cubic,TangentMode=RCTM_User,Time=0.6328,Value=0.4639,ArriveTangent=1.4308,LeaveTangent=1.4308),(Time=1,Value=1))))
    SourceMinimum=0
    TargetMinimum=0
    Value=@n9.Result
   )
   &RigVMFunction_AnimEvalRichCurve(
    Curve=(EditorCurveData=(Keys=((),(InterpMode=RCIM_Cubic,TangentMode=RCTM_User,Time=0.5108,Value=0.5138,ArriveTangent=0.9648,LeaveTangent=0.9648),(InterpMode=RCIM_Cubic,TangentMode=RCTM_User,Time=1,Value=1,ArriveTangent=0.9684,LeaveTangent=0.9684))))
    SourceMinimum=0
    TargetMinimum=0
    Value=@n5.Value
   )
   &RigVMFunction_AnimEvalRichCurve(
    Curve="(EditorCurveData=(Keys=((InterpMode=RCIM_Cubic),(InterpMode=RCIM_Cubic,TangentMode=RCTM_User,Time=0.3498,Value=0.2636,ArriveTangent=0.7793,LeaveTangent=0.7793),(InterpMode=RCIM_Cubic,TangentMode=RCTM_User,Time=0.74,Value=0.5413,ArriveTangent=0.8328,LeaveTa ...(truncated)"
    SourceMinimum=0
    TargetMinimum=0
    Value=@n11.Result
   )
   &RigVMFunction_AnimEvalRichCurve(
    Curve="(EditorCurveData=(Keys=((InterpMode=RCIM_Cubic),(InterpMode=RCIM_Cubic,TangentMode=RCTM_User,Time=0.5408,Value=0.5035,ArriveTangent=1.073,LeaveTangent=1.073),(InterpMode=RCIM_Cubic,TangentMode=RCTM_User,Time=0.9966,Value=0.9999,ArriveTangent=1.125,LeaveTan ...(truncated)"
    SourceMinimum=0
    TargetMinimum=0
    Value=@DeltaRotation_5.ouputWeight
   )
   RigVMVariableNode() {variable=Leg_L_Side}
   RigVMVariableNode() {variable=Emote}
   RigVMVariableNode() {variable=Crouching}
   RigVMVariableNode() {variable=Crouching}
   RigVMVariableNode() {variable=Leg_R_Side}
   RigVMVariableNode() {variable=Crouching}
   RigVMVariableNode(ExecuteContext=(), Value=true) {variable=CrouchingBool}
  ]}
 New Function_ContainedGraph {nodes [
   RigVMFunctionEntryNode()
   > RigVMFunction_Sequence = RigVMAggregateNode()
   n1 = &RigUnit_SphericalPoseReader(
    ActiveRegionScaleFactors=(PositiveWidth=4,NegativeWidth=1,PositiveHeight=1,NegativeHeight=4)
    ActiveRegionSize=0.1
    Debug=(bDrawDebug=False,DebugScale=25,DebugSegments=20,DebugThickness=0.25)
    DriverAxis=(X=-1,Y=0,Z=0)
    DriverItem=(Type=Bone,Name="thigh_l")
    FalloffRegionScaleFactors=(PositiveWidth=1,NegativeWidth=1.3,PositiveHeight=0.9,NegativeHeight=1)
    FalloffSize=0.48
    OptionalParentItem=(Type=Bone,Name="None")
    RotationOffset=(X=90,Y=0,Z=-17)
    ExecutePin=@RigVMFunction_Sequence.A
   )
   RigVMFunction_AnimEvalRichCurve = &RigVMFunction_AnimEvalRichCurve(
    Curve="(EditorCurveData=(Keys=((),(InterpMode=RCIM_Cubic,TangentMode=RCTM_User,Time=0.4854,Value=0.3672,ArriveTangent=1.0067,LeaveTangent=1.0067),(InterpMode=RCIM_Cubic,TangentMode=RCTM_User,Time=0.9966,Value=0.9999,ArriveTangent=1.125,LeaveTangent=1.125),(Time=1 ...(truncated)"
    SourceMinimum=0
    TargetMinimum=0
    Value=@n1.OutputParam
   )
   > RigVMVariableNode(Value=@RigVMFunction_AnimEvalRichCurve.Result) {variable=Leg_L_Front}
   > n2 = &RigUnit_SphericalPoseReader(
    ActiveRegionScaleFactors=(PositiveWidth=1,NegativeWidth=1,PositiveHeight=4,NegativeHeight=0.1)
    ActiveRegionSize=0.1
    Debug=(bDrawDebug=False,DebugScale=25,DebugSegments=20,DebugThickness=0.25)
    DriverAxis=(X=-1,Y=0,Z=0)
    DriverItem=(Type=Bone,Name="thigh_l")
    FalloffRegionScaleFactors=(PositiveWidth=1,NegativeWidth=1,PositiveHeight=1,NegativeHeight=1)
    FalloffSize=0.45
    FlipHeightScaling=false
    FlipWidthScaling=false
    OptionalParentItem=(Type=Bone,Name="None")
    RotationOffset=(X=180,Y=0,Z=90)
   )
   n3 = &RigVMFunction_AnimEvalRichCurve(
    Curve=(EditorCurveData=(Keys=((),(InterpMode=RCIM_Cubic,TangentMode=RCTM_User,Time=0.4904,Value=0.3612,ArriveTangent=1.0067,LeaveTangent=1.0067),(InterpMode=RCIM_Cubic,TangentMode=RCTM_User,Time=1,Value=1,ArriveTangent=1.2446,LeaveTangent=1.2446))))
    SourceMinimum=0
    TargetMinimum=0
    Value=@n2.OutputParam
   )
   > RigVMVariableNode(Value=@n3.Result) {variable=Leg_L_Side}
   > n4 = &RigUnit_SphericalPoseReader(
    ActiveRegionScaleFactors=(PositiveWidth=3.5,NegativeWidth=0.5,PositiveHeight=3,NegativeHeight=3)
    ActiveRegionSize=0.1
    Debug=(bDrawDebug=false,DebugScale=25,DebugSegments=20,DebugThickness=0.25)
    DriverAxis=(X=-1,Y=0,Z=0)
    DriverItem=(Type=Bone,Name="thigh_l")
    FalloffRegionScaleFactors=(PositiveWidth=1,NegativeWidth=1,PositiveHeight=1,NegativeHeight=1)
    FalloffSize=0.48
    OptionalParentItem=(Type=Bone,Name="None")
    RotationOffset=(X=-90,Y=0,Z=0)
   )
   n5 = &RigVMFunction_AnimEvalRichCurve(
    Curve="(EditorCurveData=(Keys=((),(InterpMode=RCIM_Cubic,TangentMode=RCTM_User,Time=0.3883,Value=0.2125,ArriveTangent=0.6007,LeaveTangent=0.6007),(InterpMode=RCIM_Cubic,TangentMode=RCTM_User,Time=0.7023,Value=0.518,ArriveTangent=1.0909,LeaveTangent=1.0909),(Inter ...(truncated)"
    SourceMinimum=0
    TargetMinimum=0
    Value=@n4.OutputParam
   )
   > RigVMVariableNode(Value=@n5.Result) {variable=Leg_L_Back}
   > n6 = &RigUnit_SphericalPoseReader(
    ActiveRegionScaleFactors=(PositiveWidth=1,NegativeWidth=0.1,PositiveHeight=0.05,NegativeHeight=0.05)
    ActiveRegionSize=0.2
    Debug=(bDrawDebug=False,DebugScale=25,DebugSegments=20,DebugThickness=0.25)
    DriverAxis=(X=-1,Y=0,Z=0)
    DriverItem=(Type=Bone,Name="thigh_l")
    FalloffRegionScaleFactors=(PositiveWidth=1,NegativeWidth=1,PositiveHeight=0.6,NegativeHeight=0.6)
    FalloffSize=0.4
    FlipHeightScaling=false
    FlipWidthScaling=true
    OptionalParentItem=(Type=Bone,Name="None")
    RotationOffset=(X=-135,Y=0,Z=180)
   )
   > RigVMVariableNode(Value=@n6.OutputParam) {variable=Leg_L_Front_Mix}
   > n7 = &RigUnit_SphericalPoseReader(
    ActiveRegionScaleFactors=(PositiveWidth=4,NegativeWidth=0.1,PositiveHeight=2,NegativeHeight=4)
    ActiveRegionSize=0.1
    Debug=(bDrawDebug=False,DebugScale=25,DebugSegments=20,DebugThickness=0.25)
    DriverAxis=(X=1,Y=0,Z=0)
    DriverItem=(Type=Bone,Name="thigh_r")
    FalloffRegionScaleFactors=(PositiveWidth=1,NegativeWidth=1,PositiveHeight=1.3,NegativeHeight=1)
    FalloffSize=0.48
    FlipHeightScaling=false
    FlipWidthScaling=true
    OptionalParentItem=(Type=Bone,Name="None")
    RotationOffset=(X=-110,Y=0,Z=0)
   )
   n8 = &RigVMFunction_AnimEvalRichCurve(
    Curve="(EditorCurveData=(Keys=((InterpMode=RCIM_Cubic),(InterpMode=RCIM_Cubic,TangentMode=RCTM_User,Time=0.4959,Value=0.404,ArriveTangent=1.2722,LeaveTangent=1.2722),(InterpMode=RCIM_Cubic,TangentMode=RCTM_User,Time=0.9966,Value=0.9999,ArriveTangent=1.568,LeaveTa ...(truncated)"
    SourceMinimum=0
    TargetMinimum=0
    Value=@n7.OutputParam
   )
   > RigVMVariableNode(Value=@n8.Result) {variable=Leg_R_Front}
   > n9 = &RigUnit_SphericalPoseReader(
    ActiveRegionScaleFactors=(PositiveWidth=4,NegativeWidth=0.1,PositiveHeight=1,NegativeHeight=2)
    ActiveRegionSize=0.1
    Debug=(bDrawDebug=False,DebugScale=25,DebugSegments=20,DebugThickness=0.25)
    DriverAxis=(X=1,Y=0,Z=0)
    DriverItem=(Type=Bone,Name="thigh_r")
    FalloffRegionScaleFactors=(PositiveWidth=1,NegativeWidth=0.98,PositiveHeight=1,NegativeHeight=1.31)
    FalloffSize=0.48
    FlipHeightScaling=false
    FlipWidthScaling=true
    OptionalParentItem=(Type=Bone,Name="None")
    RotationOffset=(X=18,Y=1,Z=0)
   )
   n10 = &RigVMFunction_AnimEvalRichCurve(
    Curve="(EditorCurveData=(Keys=((InterpMode=RCIM_Cubic),(InterpMode=RCIM_Cubic,TangentMode=RCTM_User,Time=0.5406,Value=0.4054,ArriveTangent=1.073,LeaveTangent=1.073),(InterpMode=RCIM_Cubic,TangentMode=RCTM_User,Time=0.9966,Value=0.9999,ArriveTangent=1.125,LeaveTan ...(truncated)"
    SourceMinimum=0
    TargetMinimum=0
    Value=@n9.OutputParam
   )
   > RigVMVariableNode(Value=@n10.Result) {variable=Leg_R_Side}
   > n11 = &RigUnit_SphericalPoseReader(
    ActiveRegionScaleFactors=(PositiveWidth=3.5,NegativeWidth=0.5,PositiveHeight=3,NegativeHeight=3)
    ActiveRegionSize=0.1
    Debug=(bDrawDebug=False,DebugScale=25,DebugSegments=20,DebugThickness=0.25)
    DriverAxis=(X=-1,Y=0,Z=0)
    DriverItem=(Type=Bone,Name="thigh_r")
    FalloffRegionScaleFactors=(PositiveWidth=1,NegativeWidth=1,PositiveHeight=1,NegativeHeight=1)
    FalloffSize=0.48
    OptionalParentItem=(Type=Bone,Name="pelvis")
    RotationOffset=(X=-90,Y=0,Z=-10)
   )
   n12 = &RigVMFunction_AnimEvalRichCurve(
    Curve="(EditorCurveData=(Keys=((InterpMode=RCIM_Cubic),(InterpMode=RCIM_Cubic,TangentMode=RCTM_User,Time=0.5406,Value=0.4054,ArriveTangent=1.073,LeaveTangent=1.073),(InterpMode=RCIM_Cubic,TangentMode=RCTM_User,Time=0.9966,Value=0.9999,ArriveTangent=1.125,LeaveTan ...(truncated)"
    SourceMinimum=0
    TargetMinimum=0
    Value=@n11.OutputParam
   )
   > RigVMVariableNode(Value=@n12.Result) {variable=Leg_R_Back}
   > n13 = &RigUnit_SphericalPoseReader(
    ActiveRegionScaleFactors=(PositiveWidth=1,NegativeWidth=0.1,PositiveHeight=0.05,NegativeHeight=0.05)
    ActiveRegionSize=0.2
    Debug=(bDrawDebug=False,DebugScale=25,DebugSegments=20,DebugThickness=0.25)
    DriverAxis=(X=1,Y=0,Z=0)
    DriverItem=(Type=Bone,Name="thigh_r")
    FalloffRegionScaleFactors=(PositiveWidth=1,NegativeWidth=1,PositiveHeight=0.6,NegativeHeight=0.6)
    FalloffSize=0.4
    FlipHeightScaling=false
    FlipWidthScaling=true
    OptionalParentItem=(Type=Bone,Name="None")
    RotationOffset=(X=-45,Y=0,Z=0)
   )
   > RigVMVariableNode(Value=@n13.OutputParam) {variable=Leg_R_Front_Mix}
   > n14 = &RigUnit_SphericalPoseReader(
    ActiveRegionScaleFactors=(PositiveWidth=4,NegativeWidth=1,PositiveHeight=4,NegativeHeight=4)
    ActiveRegionSize=0.1
    Debug=(bDrawDebug=False,DebugScale=25,DebugSegments=20,DebugThickness=0.25)
    DriverAxis=(X=-1,Y=0,Z=0)
    DriverItem=(Type=Bone,Name="thigh_l")
    FalloffRegionScaleFactors=(PositiveWidth=1,NegativeWidth=1.5,PositiveHeight=1,NegativeHeight=1)
    FalloffSize=0.48
    OptionalParentItem=(Type=Bone,Name="None")
    RotationOffset=(X=90,Y=0,Z=-40)
   )
   > RigVMVariableNode(Value=@n14.OutputParam) {variable=Leg_L_Front_Crouch}
   > n15 = &RigUnit_SphericalPoseReader(
    ActiveRegionScaleFactors=(PositiveWidth=4,NegativeWidth=1,PositiveHeight=4,NegativeHeight=4)
    ActiveRegionSize=0.1
    Debug=(bDrawDebug=false,DebugScale=25,DebugSegments=20,DebugThickness=0.25)
    DriverAxis=(X=1,Y=0,Z=0)
    DriverItem=(Type=Bone,Name="thigh_r")
    FalloffRegionScaleFactors=(PositiveWidth=1,NegativeWidth=1.5,PositiveHeight=1,NegativeHeight=1)
    FalloffSize=0.48
    FlipHeightScaling=false
    FlipWidthScaling=true
    OptionalParentItem=(Type=Bone,Name="None")
    RotationOffset=(X=-90,Y=0,Z=-40)
   )
   > RigVMVariableNode(Value=@n15.OutputParam) {variable=Leg_R_Front_Crouch}
   RigVMFunction_MathBoolToFloat = &RigVMFunction_MathBoolToFloat(Value=RigVMVariableNode() {variable=CrouchingBool}.Value)
   AlphaInterp = &RigVMFunction_AlphaInterp(InterpSpeedDecreasing=20, InterpSpeedIncreasing=10, Scale=1, bClampResult=true, Value=@RigVMFunction_MathBoolToFloat.Result)
   RigVMVariableNode(Value=@AlphaInterp.Result, ExecuteContext=@RigVMFunction_Sequence.B) {variable=Crouching}
   n16 = &RigVMFunction_AlphaInterp(
    InterpSpeedDecreasing=5
    InterpSpeedIncreasing=5
    Scale=1
    bClampResult=true
    Value=&RigVMFunction_MathBoolToFloat(Value=RigVMVariableNode() {variable=Moving2dBool}.Value).Result
   )
   > RigVMVariableNode(Value=@n16.Result) {variable=Moving2D}
   n17 = &RigVMFunction_AlphaInterp(
    InterpSpeedDecreasing=10
    InterpSpeedIncreasing=10
    Scale=1
    bClampResult=true
    Value=&RigVMFunction_MathBoolToFloat(Value=RigVMVariableNode() {variable=TacticalSprintingBool}.Value).Result
   )
   > RigVMVariableNode(Value=@n17.Result) {variable=TacticalSprinting}
   RigVMFunctionReturnNode(ExecuteContext=@RigVMFunction_Sequence.C)
  ]}
 New Function_ContainedGraph {nodes [
   n1 = RigVMFunctionEntryNode()
   RigVMFunction_MathQuaternionToEuler = &RigVMFunction_MathQuaternionToEuler(
    RotationOrder=ZYX
    Value=&RigUnit_GetRelativeTransformForItem(
     Child=(Type=Bone,Name="None")
     Parent=(Type=Bone,Name="None")
     bChildInitial=False
     bParentInitial=False
     Child.Name=@n1.Child
     Parent.Name=@n1.Parent
    ).RelativeTransform.Rotation
   )
   Remap = &RigVMFunction_MathDoubleRemap(
    TargetMaximum=1
    TargetMinimum=0
    Value=@RigVMFunction_MathQuaternionToEuler.Result.Z
    SourceMinimum=@n1.SourceMin
    SourceMaximum=@n1.SourceMax
   )
   > RigVMFunctionReturnNode(Delta Z=@Remap.Result)
  ]}
 New Function_ContainedGraph {nodes [
   n1 = RigVMFunctionEntryNode()
   Clamp = &RigVMFunction_MathIntClamp(Value=@n1.primAxis)
   RigVMFunction_MathBoolToInteger = &RigVMFunction_MathBoolToInteger(Value=@n1.onlyTwist)
   RigVMFunction_MathQuaternionSwingTwist = &RigVMFunction_MathQuaternionSwingTwist(
    TwistAxis=@n1.twistAxis
    Input=&RigUnit_GetRelativeTransformForItem(
     Child=(Type=Bone,Name="None")
     Parent=(Type=Bone,Name="None")
     bChildInitial=False
     bParentInitial=False
     Parent.Type=@n1.ParentType
     Child.Name=@n1.child
     Parent.Name=@n1.parent
    ).RelativeTransform.Rotation
   )
   RigVMFunction_MathQuaternionToEuler = &RigVMFunction_MathQuaternionToEuler(
    Value=RigVMDispatchNode(
     Values=((),())
     Index=@RigVMFunction_MathBoolToInteger.Result
     Values.0=@RigVMFunction_MathQuaternionSwingTwist.Swing
     Values.1=@RigVMFunction_MathQuaternionSwingTwist.Twist
    ) {template="DISPATCH_RigVMDispatch_SelectInt32(in Index,in Values,out Result)"}.Result
    RotationOrder=@n1.rotationOrder
   )
   DISPATCH_RigVMDispatch_SelectInt32 = RigVMDispatchNode(
    Values=(0,0,0)
    Index=@Clamp.Result
    Values.0=@RigVMFunction_MathQuaternionToEuler.Result.X
    Values.1=@RigVMFunction_MathQuaternionToEuler.Result.Y
    Values.2=@RigVMFunction_MathQuaternionToEuler.Result.Z
   ) {template="DISPATCH_RigVMDispatch_SelectInt32(in Index,in Values,out Result)"}
   Remap = &RigVMFunction_MathDoubleRemap(
    TargetMaximum=1
    TargetMinimum=0
    bClamp=False
    Value=@DISPATCH_RigVMDispatch_SelectInt32.Result
    SourceMaximum=@n1.max
    SourceMinimum=@n1.min
   )
   > RigVMFunctionReturnNode(ouputWeight=&RigVMFunction_MathDoubleClamp(Value=@Remap.Result).Result)
  ]}
 New Function_ContainedGraph {nodes [
   n1 = RigVMFunctionEntryNode()
   RigVMFunction_ControlFlowBranch = &RigVMFunction_ControlFlowBranch(ExecuteContext=@n1.ExecuteContext, Condition=RigVMVariableNode() {variable=FrontEndBool}.Value)
   RigVMFunctionReturnNode(ExecuteContext=@RigVMFunction_ControlFlowBranch.Completed)
   n2 = &RigUnit_SphericalPoseReader(
    ActiveRegionScaleFactors=(PositiveWidth=1,NegativeWidth=1,PositiveHeight=1.7,NegativeHeight=1)
    ActiveRegionSize=0.1
    Debug=(bDrawDebug=false,bDraw2D=False,bDrawLocalAxes=False,DebugScale=25,DebugSegments=20,DebugThickness=0.25)
    DriverAxis=(X=-1,Y=0,Z=0)
    DriverItem=(Type=Bone,Name="upperarm_r")
    FalloffRegionScaleFactors=(PositiveWidth=1.5,NegativeWidth=1,PositiveHeight=2.5,NegativeHeight=1.3)
    FalloffSize=0.17
    FlipHeightScaling=False
    FlipWidthScaling=False
    OptionalParentItem=(Type=Bone,Name="None")
    RotationOffset=(X=-50,Y=0,Z=-30)
    ExecutePin=@RigVMFunction_ControlFlowBranch.False
   )
   > &RigUnit_ModifyTransforms(
    ItemToModify=((Item=(Type=Bone,Name="dyn_ArmBagginessArea_r"),Transform=(Rotation=(X=-0.0119,Y=0.064,Z=0.0637,W=-0.9958),Translation=(X=0,Y=0,Z=0),Scale3D=(X=1,Y=1,Z=1))))
    WeightMinimum=0
    Weight=@n2.OutputParam
   )
   > n3 = &RigUnit_SphericalPoseReader(
    ActiveRegionScaleFactors=(PositiveWidth=2.5,NegativeWidth=3,PositiveHeight=3,NegativeHeight=3)
    ActiveRegionSize=0.16
    Debug=(bDrawDebug=false,DebugScale=25,DebugSegments=20,DebugThickness=0.25)
    DriverAxis=(X=1,Y=0,Z=0)
    DriverItem=(Type=Bone,Name="upperarm_l")
    FalloffRegionScaleFactors=(PositiveWidth=1.5,NegativeWidth=1,PositiveHeight=2,NegativeHeight=1)
    FalloffSize=0.6
    OptionalParentItem=(Type=Bone)
    RotationOffset=(X=0,Y=0,Z=0)
   )
   > RigVMVariableNode(Value=@n3.OutputParam) {variable=Arm_L_Up}
   > &RigUnit_ModifyTransforms(
    ItemToModify=((Item=(Type=Bone,Name="dyn_ArmBagginessArea_l"),Transform=(Rotation=(X=-0.0095,Y=-0.0837,Z=0.0048,W=-0.9964),Translation=(X=0,Y=0,Z=0),Scale3D=(X=1,Y=1,Z=1))))
    WeightMinimum=0
    Weight=@n3.OutputParam
   )
   > n4 = &RigUnit_SphericalPoseReader(
    ActiveRegionScaleFactors=(PositiveWidth=2.5,NegativeWidth=3,PositiveHeight=3,NegativeHeight=3)
    ActiveRegionSize=0.15
    Debug=(bDrawDebug=false,DebugScale=25,DebugSegments=20,DebugThickness=0.25)
    DriverAxis=(X=-1,Y=0,Z=0)
    DriverItem=(Type=Bone,Name="upperarm_r")
    FalloffRegionScaleFactors=(PositiveWidth=1.5,NegativeWidth=1,PositiveHeight=2,NegativeHeight=1)
    FalloffSize=0.55
    FlipHeightScaling=true
    OptionalParentItem=(Type=Bone)
    RotationOffset=(X=180,Y=0,Z=180)
   )
   > &RigUnit_ModifyTransforms(
    ItemToModify=((Item=(Type=Bone,Name="dyn_ArmBagginessArea_r"),Transform=(Rotation=(X=0.0135,Y=-0.0867,Z=0.0032,W=-0.9961),Translation=(X=0,Y=0,Z=0),Scale3D=(X=1,Y=1,Z=1))))
    WeightMinimum=0
    Weight=@n4.OutputParam
   )
   > &RigUnit_ModifyTransforms(
    ItemToModify=((Item=(Type=Bone,Name="deform_upperarm_bk_base_r"),Transform=(Rotation=(X=0,Y=0,Z=0,W=-1),Translation=(X=-6.2617,Y=5.1378,Z=6.4278),Scale3D=(X=1,Y=1,Z=1))))
    WeightMinimum=0
    Weight=&RigVMFunction_MathDoubleMul(A=@n4.OutputParam, B=RigVMVariableNode() {variable=Pistol}.Value).Result
   )
   RigVMFunction_MathBoolToFloat = &RigVMFunction_MathBoolToFloat(Value=RigVMVariableNode() {variable=OpenGlider}.Value)
   AlphaInterp = &RigVMFunction_AlphaInterp(
    InterpSpeedDecreasing=10
    InterpSpeedIncreasing=10
    Scale=1
    bClampResult=False
    bInterpResult=False
    Value=@RigVMFunction_MathBoolToFloat.Result
   )
   Multiply = &RigVMFunction_MathDoubleMul(
    B=@AlphaInterp.Result
    A=RigVMDispatchNode(BoneName=root, Default=0, Name=deformRig_r_upperarm_up_alpha) {template="DISPATCH_RigDispatch_GetAnimAttribute(in Name,in BoneName,in Default,out Value,out Found)"}.Value
   )
   > &RigUnit_ModifyTransforms(
    ItemToModify="((Item=(Type=Bone,Name=\"dyn_ArmBagginessArea_r\"),Transform=(Rotation=(X=-0.0111,Y=0.0905,Z=0.0554,W=-0.9943),Translation=(X=0,Y=0,Z=0),Scale3D=(X=1,Y=1,Z=1))),(Item=(Type=Bone,Name=\"dyn_ArmBagginessArea_l\"),Transform=(Rotation=(X=-0.0038,Y=0.1088,Z=0.0346, ...(truncated)"
    WeightMinimum=0
    Weight=@Multiply.Result
   )
   > &RigUnit_ModifyTransforms(
    ItemToModify="((Item=(Type=Bone,Name=\"dyn_ArmBagginessArea_r\"),Transform=(Rotation=(X=0,Y=0,Z=0.054,W=-0.9985),Translation=(X=0.2107,Y=1.9436,Z=0),Scale3D=(X=1,Y=1,Z=1))),(Item=(Type=Bone,Name=\"dyn_ArmBagginessArea_l\"),Transform=(Rotation=(X=0,Y=0,Z=0,W=1),Translation=( ...(truncated)"
    WeightMinimum=0
    Weight=&RigVMFunction_MathDoubleMul(A=@AlphaInterp.Result, B=RigVMVariableNode() {variable=Move_Back}.Value).Result
   )
   n5 = &RigVMFunction_AlphaInterp(
    InterpSpeedDecreasing=10
    InterpSpeedIncreasing=10
    Scale=1
    bClampResult=true
    Value=&RigVMFunction_MathBoolToFloat(Value=RigVMVariableNode() {variable=FE_Locker_Idle}.Value).Result
   )
   &RigUnit_ModifyTransforms(
    ItemToModify="((Item=(Type=Bone,Name=\"dyn_up_ArmBagginess_l\"),Transform=(Rotation=(X=0,Y=0.0206,Z=0,W=0.9998),Translation=(X=0,Y=0,Z=0),Scale3D=(X=1,Y=1,Z=1))),(Item=(Type=Bone,Name=\"dyn_ArmBagginessArea_r\"),Transform=(Rotation=(X=0,Y=0.0509,Z=0,W=-0.9987),Translation=( ...(truncated)"
    WeightMinimum=0
    Weight=@n5.Result
    ExecutePin=@RigVMFunction_ControlFlowBranch.True
   )
   > n6 = &RigUnit_SphericalPoseReader(
    ActiveRegionScaleFactors=(PositiveWidth=1,NegativeWidth=1,PositiveHeight=1.7,NegativeHeight=1)
    ActiveRegionSize=0.1
    Debug=(bDrawDebug=True,bDraw2D=False,bDrawLocalAxes=False,DebugScale=25,DebugSegments=20,DebugThickness=0.25)
    DriverAxis=(X=1,Y=0,Z=0)
    DriverItem=(Type=Bone,Name="upperarm_l")
    FalloffRegionScaleFactors=(PositiveWidth=1.5,NegativeWidth=1,PositiveHeight=2.5,NegativeHeight=1.3)
    FalloffSize=0.17
    FlipHeightScaling=False
    FlipWidthScaling=False
    OptionalParentItem=(Type=Bone,Name="None")
    RotationOffset=(X=120,Y=30,Z=0)
   )
   > &RigUnit_ModifyTransforms(
    ItemToModify="((Item=(Type=Bone,Name=\"dyn_ArmBagginessArea_l\"),Transform=(Rotation=(X=0.0046,Y=0.0197,Z=-0.151,W=0.9883),Translation=(X=0,Y=0,Z=0),Scale3D=(X=1,Y=1,Z=1))),(Item=(Type=Bone,Name=\"dyn_ArmBagginessArea_r\"),Transform=(Rotation=(X=0.0109,Y=-0.0115,Z=0.0564,W= ...(truncated)"
    WeightMinimum=0
    Weight=@n6.OutputParam
   )
  ]}
 New Function_ContainedGraph {nodes [
   RigVMFunctionEntryNode()
   > DeltaRotation Z = RigVMFunctionReferenceNode(Child=foot_l, Parent=calf_l, SourceMax=60, SourceMin=15)
   > &RigUnit_ModifyTransforms(
    ItemToModify=((Item=(Type=Bone,Name="fix_bk_PantBagginess_l"),Transform=(Rotation=(X=0,Y=0,Z=0,W=-1),Translation=(X=6.2078,Y=0.3461,Z=-10.2798),Scale3D=(X=1,Y=1,Z=1))))
    WeightMinimum=0
    Weight=@"DeltaRotation Z".Delta Z
   )
   > DeltaRotation Z_1 = RigVMFunctionReferenceNode(Child=foot_r, Parent=calf_r, SourceMax=60, SourceMin=15)
   > &RigUnit_ModifyTransforms(
    ItemToModify=((Item=(Type=Bone,Name="fix_bk_PantBagginess_r"),Transform=(Rotation=(X=0,Y=0,Z=0,W=1),Translation=(X=-5.6146,Y=-0.3202,Z=7.3574),Scale3D=(X=1,Y=1,Z=1))))
    WeightMinimum=0
    Weight=@"DeltaRotation Z_1".Delta Z
   )
   > RigVMFunctionReturnNode()
  ]}
 New Function_ContainedGraph {nodes [
   n1 = RigVMFunctionEntryNode()
   RigVMFunction_Sequence = &RigVMFunction_Sequence(ExecuteContext=@n1.ExecuteContext)
   n2 = &RigUnit_SphericalPoseReader(
    ActiveRegionScaleFactors=(PositiveWidth=0.5,NegativeWidth=0.5,PositiveHeight=0.5,NegativeHeight=0.5)
    ActiveRegionSize=0.2
    Debug=(bDrawDebug=false,bDraw2D=false,DebugScale=25,DebugSegments=20,DebugThickness=0.25)
    DriverAxis=(X=-1,Y=0,Z=0)
    DriverItem=(Type=Bone,Name="calf_l")
    FalloffRegionScaleFactors=(PositiveWidth=0,NegativeWidth=0.7,PositiveHeight=0.7,NegativeHeight=0.7)
    FalloffSize=0.45
    OptionalParentItem=(Type=Bone,Name="thigh_l")
    RotationOffset=(X=-90,Y=0,Z=60)
    ExecutePin=@RigVMFunction_Sequence.A
   )
   > &RigUnit_ModifyTransforms(
    ItemToModify=((Item=(Type=Bone,Name="calf_counter_l"),Transform=(Rotation=(X=0,Y=0,Z=-0.1779,W=0.9841),Translation=(X=-1.212,Y=0.8652,Z=-0.0331),Scale3D=(X=1,Y=1,Z=1.1))))
    WeightMinimum=0
    Weight=@n2.OutputParam
   )
   > &RigUnit_ModifyTransforms(
    ItemToModify=((Item=(Type=Bone,Name="thigh_counter_l"),Transform=(Rotation=(X=0,Y=0,Z=0.0651,W=0.9979),Translation=(X=0.3847,Y=0.3454,Z=-0.038),Scale3D=(X=1,Y=1,Z=1))))
    WeightMinimum=0
    Weight=@n2.OutputParam
   )
   > n3 = &RigUnit_SphericalPoseReader(
    ActiveRegionScaleFactors=(PositiveWidth=0.5,NegativeWidth=0.5,PositiveHeight=0.5,NegativeHeight=0.5)
    ActiveRegionSize=0.2
    Debug=(bDrawDebug=false,bDraw2D=false,DebugScale=25,DebugSegments=20,DebugThickness=0.25)
    DriverAxis=(X=1,Y=0,Z=0)
    DriverItem=(Type=Bone,Name="calf_r")
    FalloffRegionScaleFactors=(PositiveWidth=0,NegativeWidth=0.7,PositiveHeight=0.7,NegativeHeight=0.7)
    FalloffSize=0.45
    OptionalParentItem=(Type=Bone,Name="thigh_r")
    RotationOffset=(X=-90,Y=0,Z=-120)
   )
   > &RigUnit_ModifyTransforms(
    ItemToModify=((Item=(Type=Bone,Name="calf_counter_r"),Transform=(Rotation=(X=0,Y=0,Z=-0.0096,W=1),Translation=(X=-0.0921,Y=-0.7558,Z=0),Scale3D=(X=1,Y=0.8,Z=1))))
    WeightMinimum=0
    Weight=@n3.OutputParam
   )
   > &RigUnit_ModifyTransforms(
    ItemToModify=((Item=(Type=Bone,Name="thigh_counter_r"),Transform=(Rotation=(X=0,Y=0,Z=-0.0707,W=-0.9975),Translation=(X=0,Y=0,Z=0),Scale3D=(X=1,Y=1,Z=1))))
    WeightMinimum=0
    Weight=@n3.OutputParam
   )
   RigVMFunctionReturnNode(ExecuteContext=@RigVMFunction_Sequence.B)
  ]}
 RigVMFunctionLibrary {nodes [
   DeltaRotation = RigVMCollapseNode(ExecuteContext=(), debugRotation=False, onlyTwist=False, twistAxis=(X=0,Y=0,Z=0))
   DeltaRotation Z = RigVMCollapseNode(ExecuteContext=())
   FixLegCollisionOnCrouch = RigVMCollapseNode(ExecuteContext=())
   SPRs Leg and InterpolateVariables = RigVMCollapseNode(ExecuteContext=())
   SetupArms = RigVMCollapseNode(ExecuteContext=())
   SetupLeg = RigVMCollapseNode(ExecuteContext=())
   SetupShirt = RigVMCollapseNode(ExecuteContext=())
  ]}
 RigVMFunction_Sequence_1_ContainedGraph {nodes [
   n1 = RigVMFunctionEntryNode()
   RigVMFunction_Sequence = &RigVMFunction_Sequence(ExecuteContext=@n1.ExecuteContext)
   n2 = &RigVMFunction_Sequence(ExecuteContext=@RigVMFunction_Sequence.B)
   RigVMFunctionReturnNode(A=@RigVMFunction_Sequence.A, B=@n2.A, C=@n2.B)
  ]}
 RigVMFunction_Sequence_2_ContainedGraph {nodes [
   n1 = RigVMFunctionEntryNode()
   RigVMFunction_Sequence = &RigVMFunction_Sequence(ExecuteContext=@n1.ExecuteContext)
   n2 = &RigVMFunction_Sequence(ExecuteContext=@RigVMFunction_Sequence.B)
   RigVMFunctionReturnNode(A=@RigVMFunction_Sequence.A, B=@n2.A, C=@n2.B)
  ]}
 RigVMModel {nodes [
   n1 = &RigUnit_PrepareForExecution()
   &RigUnit_HierarchyImportFromSkeleton(NameSpace=None, bIncludeCurves=False, bIncludeMeshSockets=False, bIncludeVirtualBones=true, ExecutePin=@n1.ExecutePin)
   > addNullObject = RigVMFunctionReferenceNode(
    Custom_Joints=()
    Options_Null="(pelvis_1_#1=False,spine_3_#2=False,clavicle_5_#3=False,arm_7_#4=False,leg_9_#5=true,head_11_46DC15EF465CD ...(truncated)"
   )
   RigUnit_BeginExecution = &RigUnit_BeginExecution()
   FN_DeformRig_FNC = RigVMFunctionReferenceNode(
    DeformRig_ Curve="(deformRig_l_upperarm_up_alpha_72_#6=false,deformRig_r_upperarm_up_alpha_73_#7=true,deformRig_l_upperarm_fr_alpha_74_#8=False,deformRig_r_upperarm_fr_alpha_75_CB51B80 ...(truncated)"
    DeformRig_Bool="(upperarm_1_#9=True,elbow_3_#10=True,wrist_7_#11=True,groin_8_#12=True,glute_17_#13=True,knee_9_40C408324CA261 ...(truncated)"
    DeformRig_Debug=(upperarm_debug_18_#9=False,wrist_debug_20_#11=False,lat_debug_26_#14=False,groinGlute_debug_27_#12=False)
    DeformRig_Value="(upperarm_up_18_#15=(Rotation=(X=0,Y=0,Z=0,W=1),Translation=(X=0,Y=0,Z=1),Scale3D=(X=0.7,Y=1,Z=1)),upperarm_fwd_fr_19_#16=(Rotation=(X=0,Y=0,Z=0,W=1),Translation=(X=0,Y=10,Z=0),Scale3D=(X=0.5,Y=1,Z= ...(truncated)"
    ExecuteContext=@RigUnit_BeginExecution.ExecutePin
   )
   > SetupArms = RigVMFunctionReferenceNode()
   > SetupShirt = RigVMFunctionReferenceNode()
   > SetupLeg = RigVMFunctionReferenceNode()
   > FixLegCollisionOnCrouch = RigVMFunctionReferenceNode()
   addFKCtrl = RigVMFunctionReferenceNode(
    Custom_FKCtrl=()
    ExecuteContext=()
    Options_FKCtrl="(pelvis_1_#1=true,spine_3_#2=true,clavicle_5_#3=true,arm_7_#4=true,leg_9_#5=true,head_11_46DC15EF465CD2979 ...(truncated)"
    Shape_Settings="(pelvis_2_#17=(bVisible=True,Name=\"Hexagon_Thin\",Color=(R=0.0117,G=0.8281,B=1,A=1),Transform=(Rotation=(X=0,Y=-0.7071,Z=0,W=0.7071),Translation=(X=2,Y=0,Z=0),Scale3D=(X=5,Y=5,Z=7))),spine_4_#18=(bVi ...(truncated)"
   )
   setFKCtrl = RigVMFunctionReferenceNode(ExecuteContext=(), setFK_Ctrl=false)
  ]}
]
variables [
 Arm_L_Up {type=double, default=0}
 Crouching {type=double, default=0}
 CrouchingBool {type=bool, default=False}
 DBNO {type=bool, default=False}
 Emote {type=double, default=0}
 FE_Locker_Idle {type=bool, default=False}
 FrontEndBool {type=bool, default=False}
 Leg_L_Back {type=double, default=0}
 Leg_L_Front {type=double, default=0}
 Leg_L_Front_Crouch {type=double, default=0}
 Leg_L_Front_Mix {type=double, default=0}
 Leg_L_Side {type=double, default=0}
 Leg_R_Back {type=double, default=0}
 Leg_R_Front {type=double, default=0}
 Leg_R_Front_Crouch {type=double, default=0}
 Leg_R_Front_Mix {type=double, default=0}
 Leg_R_Side {type=double, default=0}
 Move_Back {type=double, default=0}
 Moving2D {type=double, default=0}
 Moving2dBool {type=bool, default=False}
 OpenGlider {type=bool, default=False}
 Pistol {type=double, default=0}
 TacticalSprinting {type=double, default=0}
 TacticalSprintingBool {type=bool, default=False}
]
