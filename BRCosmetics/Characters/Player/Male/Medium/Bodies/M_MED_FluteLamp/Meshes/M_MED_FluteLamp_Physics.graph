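# Physics asset for the M_MED_FluteLamp male player body, set up for secondary motion of the baggy shirt.
# Capsules on the pelvis and thighs form the body volumes, with tiny capsules on the shirt's base bones as chain roots.
# Two-link capsule chains at back left, center, right and both sides hang from those roots with tight swing limits and locked twist.
# Thighs and some base bones link to the pelvis with free angular constraints; the front base has no chain.
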PhysicsAsset /BRCosmetics/Characters/Player/Male/Medium/Bodies/M_MED_FluteLamp/Meshes/M_MED_FluteLamp_Physics
bodies {
 base_bk_ShirtBagginess_c [{shapeType=Capsule, length=1, radius=1, shapeName=base_bk_ShirtBagginess_r_capsule, rotation=<rot 0 0 -180>}]
 base_bk_ShirtBagginess_l [{shapeType=Capsule, length=1, radius=1, shapeName=base_bk_ShirtBagginess_r_capsule, rotation=<rot 0 0 -180>}]
 base_bk_ShirtBagginess_r [{shapeType=Capsule, length=1, radius=1, shapeName=base_bk_ShirtBagginess_l_capsule}]
 base_fr_ShirtBagginess_c [{shapeType=Capsule, length=0.505, radius=0.505, shapeName=base_fr_ShirtBagginess_c_capsule, center=<-0.5382 0 0>, rotation=<rot 0 -90 0>}]
 base_sd_ShirtBagginess_l [{shapeType=Capsule, length=0.505, radius=0.505, shapeName=base_sd_ShirtBagginess_l_capsule, rotation=<rot 0 -90 0>}]
 base_sd_ShirtBagginess_r [{shapeType=Capsule, length=0.505, radius=0.505, shapeName=base_sd_ShirtBagginess_r_capsule, rotation=<rot 0 -90 0>}]
 dyn_bk_ShirtBagginess_c_02 [{shapeType=Capsule, length=9.4815, radius=4.471, shapeName=dyn_bk_ShirtBagginess_l_02_capsule, center=<1 0 5>, rotation=<rot 0 0 90>}]
 dyn_bk_ShirtBagginess_c_03 [{shapeType=Capsule, length=9.4815, radius=4.471, shapeName=dyn_bk_ShirtBagginess_l_02_capsule, center=<1 0 5>, rotation=<rot 0 0 90>}]
 dyn_bk_ShirtBagginess_l_02 [{shapeType=Capsule, length=9.4815, radius=4.471, shapeName=dyn_bk_ShirtBagginess_l_02_capsule, center=<1 0 5>, rotation=<rot 0 0 90>}]
 dyn_bk_ShirtBagginess_l_03 [{shapeType=Capsule, length=9.4815, radius=4.471, shapeName=dyn_bk_ShirtBagginess_l_02_capsule, center=<1 0 5>, rotation=<rot 0 0 90>}]
 dyn_bk_ShirtBagginess_r_02 [{shapeType=Capsule, length=9.4815, radius=4.471, shapeName=dyn_bk_ShirtBagginess_r_02_capsule, center=<-1 0 -5>, rotation=<rot 0 0 -90>}]
 dyn_bk_ShirtBagginess_r_03 [{shapeType=Capsule, length=9.4815, radius=4.471, shapeName=dyn_bk_ShirtBagginess_r_02_capsule, center=<-1 0 -5>, rotation=<rot 0 0 -90>}]
 dyn_sd_ShirtBagginess_l_01 [{shapeType=Capsule, length=9.4815, radius=4.471, shapeName=dyn_bk_ShirtBagginess_l_02_capsule, center=<1 0 6>, rotation=<rot 0 0 90>}]
 dyn_sd_ShirtBagginess_l_02 [{shapeType=Capsule, length=9.4815, radius=4.471, shapeName=dyn_bk_ShirtBagginess_l_02_capsule, center=<1 0 6>, rotation=<rot 0 0 90>}]
 dyn_sd_ShirtBagginess_r_01 [{shapeType=Capsule, length=8.9115, radius=5.041, shapeName=dyn_bk_ShirtBagginess_r_02_capsule, center=<-1 0 -6>, rotation=<rot 0 0 -90>}]
 dyn_sd_ShirtBagginess_r_02 [{
   shapeType=Capsule
   length=10.4264
   radius=5.6846
   shapeName=dyn_bk_ShirtBagginess_r_02_capsule
   center=<-1 0 -7.3345>
   rotation=<rot 0 0 -90>
  }]
 pelvis [
  {shapeType=Capsule, length=11.0463, radius=10.0389, shapeName=pelvis_capsule, center=<0 -0.7854 -0.2143>}
  {shapeType=Capsule, length=7.7348, radius=11.7882, shapeName=pelvis_capsule, center=<-8.1307 0.5352 0>}
 ]
 thigh_l [{shapeType=Capsule, length=21.7192, radius=8.2731, shapeName=thigh_l_capsule, center=<-10.5262 -0.016 1.2693>, rotation=<rot 90 0 0>}]
 thigh_r [{shapeType=Capsule, length=12.6417, radius=8.4715, shapeName=thigh_r_capsule, center=<10.952 3.4379 -1.4393>, rotation=<rot 90 172.9072 0>}]
}
constraints [
 {
  bone1Name=base_sd_ShirtBagginess_l
  bone2Name=pelvis
  swing1LimitDegrees=45
  swing1Motion=Free
  swing2LimitDegrees=45
  swing2Motion=Free
  twistLimitDegrees=45
  twistMotion=Free
 }
 {
  bone1Name=thigh_l
  bone2Name=pelvis
  swing1LimitDegrees=45
  swing1Motion=Free
  swing2LimitDegrees=45
  swing2Motion=Free
  twistLimitDegrees=45
  twistMotion=Free
 }
 {
  bone1Name=base_fr_ShirtBagginess_c
  bone2Name=pelvis
  swing1LimitDegrees=45
  swing1Motion=Free
  swing2LimitDegrees=45
  swing2Motion=Free
  twistLimitDegrees=45
  twistMotion=Free
 }
 {
  bone1Name=base_bk_ShirtBagginess_r
  bone2Name=pelvis
  swing1LimitDegrees=45
  swing1Motion=Free
  swing2LimitDegrees=45
  swing2Motion=Free
  twistLimitDegrees=45
  twistMotion=Free
 }
 {
  bone1Name=thigh_r
  bone2Name=pelvis
  swing1LimitDegrees=45
  swing1Motion=Free
  swing2LimitDegrees=45
  swing2Motion=Free
  twistLimitDegrees=45
  twistMotion=Free
 }
 {
  bone1Name=dyn_bk_ShirtBagginess_l_02
  bone2Name=base_bk_ShirtBagginess_l
  swing1LimitDegrees=3
  swing2LimitDegrees=8
  twistLimitDegrees=45
  twistMotion=Locked
 }
 {
  bone1Name=dyn_bk_ShirtBagginess_l_03
  bone2Name=dyn_bk_ShirtBagginess_l_02
  swing1LimitDegrees=3
  swing2LimitDegrees=8
  twistLimitDegrees=45
  twistMotion=Locked
 }
 {
  bone1Name=dyn_bk_ShirtBagginess_r_02
  bone2Name=base_bk_ShirtBagginess_r
  swing1LimitDegrees=3
  swing2LimitDegrees=8
  twistLimitDegrees=45
  twistMotion=Locked
 }
 {
  bone1Name=dyn_bk_ShirtBagginess_r_03
  bone2Name=dyn_bk_ShirtBagginess_r_02
  swing1LimitDegrees=3
  swing2LimitDegrees=8
  twistLimitDegrees=45
  twistMotion=Locked
 }
 {
  bone1Name=dyn_bk_ShirtBagginess_c_02
  bone2Name=base_bk_ShirtBagginess_c
  swing1LimitDegrees=3
  swing2LimitDegrees=8
  twistLimitDegrees=45
  twistMotion=Locked
 }
 {
  bone1Name=dyn_bk_ShirtBagginess_c_03
  bone2Name=dyn_bk_ShirtBagginess_c_02
  swing1LimitDegrees=3
  swing2LimitDegrees=8
  twistLimitDegrees=45
  twistMotion=Locked
 }
 {
  bone1Name=dyn_sd_ShirtBagginess_l_01
  bone2Name=base_sd_ShirtBagginess_l
  swing1LimitDegrees=3
  swing2LimitDegrees=10
  twistLimitDegrees=45
  twistMotion=Locked
 }
 {
  bone1Name=dyn_sd_ShirtBagginess_l_02
  bone2Name=dyn_sd_ShirtBagginess_l_01
  swing1LimitDegrees=3
  swing2LimitDegrees=10
  twistLimitDegrees=45
  twistMotion=Locked
 }
 {
  bone1Name=dyn_sd_ShirtBagginess_r_01
  bone2Name=base_sd_ShirtBagginess_r
  swing1LimitDegrees=3
  swing2LimitDegrees=10
  twistLimitDegrees=45
  twistMotion=Locked
 }
 {
  bone1Name=dyn_sd_ShirtBagginess_r_02
  bone2Name=dyn_sd_ShirtBagginess_r_01
  swing1LimitDegrees=3
  swing2LimitDegrees=10
  twistLimitDegrees=45
  twistMotion=Locked
 }
]
