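# Post-process style AnimBP for the M_MED_FluteLamp player body, adding secondary motion to baggy shirt, pants and sleeves.
# Copied parent pose runs through Control Rig, RigidBody, shirt/pant AnimDynamics and additive bone offsets, using cached poses.
# Two differently tuned sleeve AnimDynamics chains blend by front-end locker idle state.
# EventGraph derives movement, pistol, emote and glider inputs, computes gravity override, and drives material WPO cloth, adding wind while parachuting or skydive-floating.
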AnimBlueprint /BRCosmetics/Characters/Player/Male/Medium/Bodies/M_MED_FluteLamp/Meshes/M_MED_FluteLamp_AnimBP
target_skeleton=&/M_MED_FluteLamp_Skeleton
graphs [
 AnimGraph {nodes [
   n1 = Anim Dynamics - Bone: dyn_lw_ArmBagginess_r!(ComponentPose=Local To Component(LocalPose=Use cached pose 'FixedCache'()), Alpha=1, GravityScale:float=1, ExternalForce="0,0,0") -> (Pose) {
    angularDampingOverride=1
    angularSpringConstant=1000
    bAngularSpring=true
    bGravityOverrideInSimSpace=true
    bOverrideAngularDamping=true
    bUseGravityOverride=true
    simulationSpace=RootRelative
    alphaBoolBlend {blendOption=Linear}
    boundBone {boneName=dyn_lw_ArmBagginess_r}
    componentAppliedLinearAccClamp=<3000 3000 3000>
    componentLinearAccScale=<0.3 0.3 0.3>
    componentLinearVelScale=<0.3 0.3 0.3>
    gravityOverride {z=-500}
    physicsBodyDefinitions [{
      sphereCollisionRadius=0
      boundBone {boneName=dyn_lw_ArmBagginess_r}
      boxExtents=<10 20 20>
      constraintSetup {angularLimitsMax=<3 2 10>, angularLimitsMin=<-3 -15 -10>, angularTarget=<0 0 0>}
      localJointOffset=<8 0 0>
     }]
    relativeSpaceBone {boneName=clavicle_r}
    simSpaceSettings {simSpaceAngularAlpha=0.2}
   }
   n2 = Anim Dynamics - Bone: dyn_fr_ArmBagginess_r!(ComponentPose=@n1, Alpha=1, GravityScale:float=1, ExternalForce="0,0,0") -> (Pose) {
    angularDampingOverride=1
    angularSpringConstant=1000
    bGravityOverrideInSimSpace=true
    bOverrideAngularDamping=true
    bUseGravityOverride=true
    simulationSpace=RootRelative
    alphaBoolBlend {blendOption=Linear}
    boundBone {boneName=dyn_fr_ArmBagginess_r}
    componentAppliedLinearAccClamp=<3000 3000 3000>
    componentLinearAccScale=<0.3 0.3 0.3>
    componentLinearVelScale=<0.3 0.3 0.3>
    gravityOverride {z=-500}
    physicsBodyDefinitions [{
      sphereCollisionRadius=0
      boundBone {boneName=dyn_fr_ArmBagginess_r}
      boxExtents=<10 20 20>
      constraintSetup {angularLimitsMax=<3 5 2>, angularLimitsMin=<-3 -5 -2>, angularTarget=<0 0 0>}
      localJointOffset=<8 1 2>
     }]
    relativeSpaceBone {boneName=upperarm_r}
    simSpaceSettings {simSpaceAngularAlpha=0.2}
   }
   n3 = Anim Dynamics - Bone: dyn_bk_ArmBagginess_r!(ComponentPose=@n2, Alpha=1, GravityScale:float=1, ExternalForce="0,0,0") -> (Pose) {
    angularDampingOverride=1
    angularSpringConstant=1000
    bGravityOverrideInSimSpace=true
    bOverrideAngularDamping=true
    bUseGravityOverride=true
    simulationSpace=RootRelative
    alphaBoolBlend {blendOption=Linear}
    boundBone {boneName=dyn_bk_ArmBagginess_r}
    componentAppliedLinearAccClamp=<3000 3000 3000>
    componentLinearAccScale=<0.3 0.3 0.3>
    componentLinearVelScale=<0.3 0.3 0.3>
    gravityOverride {z=-500}
    physicsBodyDefinitions [{
      sphereCollisionRadius=0
      boundBone {boneName=dyn_bk_ArmBagginess_r}
      boxExtents=<10 20 20>
      constraintSetup {angularLimitsMax=<3 3 2>, angularLimitsMin=<-3 -5 -2>, angularTarget=<0 0 0>}
      localJointOffset=<8 -1 2>
     }]
    relativeSpaceBone {boneName=upperarm_r}
    simSpaceSettings {simSpaceAngularAlpha=0.2}
   }
   n4 = Anim Dynamics - Bone: dyn_up_ArmBagginess_r!(ComponentPose=@n3, Alpha=1, GravityScale:float=1, ExternalForce="0,0,0") -> (Pose) {
    angularDampingOverride=1
    angularSpringConstant=1000
    bGravityOverrideInSimSpace=true
    bOverrideAngularDamping=true
    bUseGravityOverride=true
    simulationSpace=BoneRelative
    alphaBoolBlend {blendOption=Linear}
    boundBone {boneName=dyn_up_ArmBagginess_r}
    componentAppliedLinearAccClamp=<3000 3000 3000>
    componentLinearAccScale=<0.3 0.3 0.3>
    componentLinearVelScale=<0.3 0.3 0.3>
    gravityOverride {y=500, z=50}
    physicsBodyDefinitions [{
      sphereCollisionRadius=0
      boundBone {boneName=dyn_up_ArmBagginess_r}
      boxExtents=<10 20 20>
      constraintSetup {angularLimitsMax=<0 2 0>, angularLimitsMin=<0 -5 -5>, angularTarget=<0 0 0>}
      localJointOffset=<6 0 0>
     }]
    relativeSpaceBone {boneName=upperarm_twist_01_r}
    simSpaceSettings {simSpaceAngularAlpha=0.2}
   }
   n5 = Anim Dynamics - Bone: dyn_lw_ArmBagginess_l!(ComponentPose=@n4, Alpha=1, GravityScale:float=1, ExternalForce="0,0,0") -> (Pose) {
    angularDampingOverride=1
    angularSpringConstant=1000
    bAngularSpring=true
    bGravityOverrideInSimSpace=true
    bOverrideAngularDamping=true
    bUseGravityOverride=true
    simulationSpace=RootRelative
    alphaBoolBlend {blendOption=Linear}
    boundBone {boneName=dyn_lw_ArmBagginess_l}
    componentAppliedLinearAccClamp=<3000 3000 3000>
    componentLinearAccScale=<0.2 0.2 0.2>
    componentLinearVelScale=<0.3 0.3 0.3>
    gravityOverride {z=-500}
    physicsBodyDefinitions [{
      sphereCollisionRadius=0
      boundBone {boneName=dyn_lw_ArmBagginess_l}
      boxExtents=<20 20 10>
      constraintSetup {angularLimitsMax=<3 2 10>, angularLimitsMin=<-3 -15 -10>, angularTarget=<0 0 0>}
      localJointOffset=<8 0 0>
     }]
    relativeSpaceBone {boneName=lowerarm_r}
    simSpaceSettings {simSpaceAngularAlpha=0.2}
   }
   n6 = Anim Dynamics - Bone: dyn_fr_ArmBagginess_l!(ComponentPose=@n5, Alpha=1, GravityScale:float=1, ExternalForce="0,0,0") -> (Pose) {
    angularDampingOverride=1
    angularSpringConstant=1000
    bGravityOverrideInSimSpace=true
    bOverrideAngularDamping=true
    bUseGravityOverride=true
    simulationSpace=RootRelative
    alphaBoolBlend {blendOption=Linear}
    boundBone {boneName=dyn_fr_ArmBagginess_l}
    componentAppliedLinearAccClamp=<3000 3000 3000>
    componentLinearAccScale=<0.1 0.1 0.1>
    componentLinearVelScale=<0.3 0.3 0.3>
    gravityOverride {z=-500}
    physicsBodyDefinitions [{
      sphereCollisionRadius=0
      boundBone {boneName=dyn_fr_ArmBagginess_l}
      boxExtents=<10 20 20>
      constraintSetup {angularLimitsMax=<3 5 2>, angularLimitsMin=<-3 -5 -2>, angularTarget=<0 0 0>}
      localJointOffset=<8 1 2>
     }]
    relativeSpaceBone {boneName=upperarm_r}
    simSpaceSettings {simSpaceAngularAlpha=0.2}
   }
   n7 = Anim Dynamics - Bone: dyn_bk_ArmBagginess_l!(ComponentPose=@n6, Alpha=1, GravityScale:float=1, ExternalForce="0,0,0") -> (Pose) {
    angularDampingOverride=1
    angularSpringConstant=1000
    bGravityOverrideInSimSpace=true
    bOverrideAngularDamping=true
    bUseGravityOverride=true
    simulationSpace=RootRelative
    alphaBoolBlend {blendOption=Linear}
    boundBone {boneName=dyn_bk_ArmBagginess_l}
    componentAppliedLinearAccClamp=<3000 3000 3000>
    componentLinearAccScale=<0.1 0.1 0.1>
    componentLinearVelScale=<0.3 0.3 0.3>
    gravityOverride {z=-500}
    physicsBodyDefinitions [{
      sphereCollisionRadius=0
      boundBone {boneName=dyn_bk_ArmBagginess_l}
      boxExtents=<10 20 20>
      constraintSetup {angularLimitsMax=<3 5 2>, angularLimitsMin=<-3 -5 -2>, angularTarget=<0 0 0>}
      localJointOffset=<8 -1 2>
     }]
    relativeSpaceBone {boneName=upperarm_r}
    simSpaceSettings {simSpaceAngularAlpha=0.2}
   }
   n8 = Anim Dynamics - Bone: dyn_up_ArmBagginess_l!(ComponentPose=@n7, Alpha=1, GravityScale:float=1, ExternalForce="0,0,0") -> (Pose) {
    angularDampingOverride=1
    angularSpringConstant=1000
    bGravityOverrideInSimSpace=true
    bOverrideAngularDamping=true
    bUseGravityOverride=true
    simulationSpace=BoneRelative
    alphaBoolBlend {blendOption=Linear}
    boundBone {boneName=dyn_up_ArmBagginess_l}
    componentAppliedLinearAccClamp=<3000 3000 3000>
    componentLinearAccScale=<0.1 0.1 0.1>
    componentLinearVelScale=<0.2 0.2 0.2>
    gravityOverride {y=500, z=-50}
    physicsBodyDefinitions [{
      sphereCollisionRadius=0
      boundBone {boneName=dyn_up_ArmBagginess_l}
      boxExtents=<15 20 20>
      constraintSetup {angularLimitsMax=<0 2 5>, angularLimitsMin=<0 -5 -5>, angularTarget=<0 0 0>}
      localJointOffset=<8 0 0>
     }]
    relativeSpaceBone {boneName=upperarm_twist_01_l}
    simSpaceSettings {simSpaceAngularAlpha=0.2}
   }
   n9 = Anim Dynamics - Bone: dyn_lw_ArmBagginess_r(ComponentPose=Local To Component(LocalPose=Use cached pose 'FixedCache'()), Alpha=1, GravityScale=1, ExternalForce="0,0,0") {
    angularDampingOverride=0.99
    angularSpringConstant=1000
    bAngularSpring=true
    bGravityOverrideInSimSpace=true
    bOverrideAngularDamping=true
    bUseGravityOverride=true
    simulationSpace=RootRelative
    alphaBoolBlend {blendOption=Linear}
    boundBone {boneName=dyn_lw_ArmBagginess_r}
    componentAppliedLinearAccClamp=<3000 3000 3000>
    componentLinearAccScale=<0.2 0.2 0.2>
    componentLinearVelScale=<0.3 0.3 0.3>
    gravityOverride {z=-500}
    physicsBodyDefinitions [{
      sphereCollisionRadius=0
      boundBone {boneName=dyn_lw_ArmBagginess_r}
      boxExtents=<20 20 10>
      constraintSetup {angularLimitsMax=<3 2 10>, angularLimitsMin=<-3 -15 -10>, angularTarget=<0 0 0>}
      localJointOffset=<8 0 0>
     }]
    relativeSpaceBone {boneName=lowerarm_r}
   }
   n10 = Anim Dynamics - Bone: dyn_fr_ArmBagginess_r(ComponentPose=@n9, Alpha=1, GravityScale=1, ExternalForce="0,0,0") {
    angularDampingOverride=0.9999
    angularSpringConstant=500
    bGravityOverrideInSimSpace=true
    bOverrideAngularDamping=true
    bUseGravityOverride=true
    simulationSpace=RootRelative
    alphaBoolBlend {blendOption=Linear}
    boundBone {boneName=dyn_fr_ArmBagginess_r}
    componentAppliedLinearAccClamp=<3000 3000 3000>
    componentLinearAccScale=<0.1 0.1 0.1>
    componentLinearVelScale=<0.3 0.3 0.3>
    gravityOverride {z=-500}
    physicsBodyDefinitions [{
      sphereCollisionRadius=0
      boundBone {boneName=dyn_fr_ArmBagginess_r}
      boxExtents=<10 20 20>
      constraintSetup {angularLimitsMax=<3 5 2>, angularLimitsMin=<-3 -5 -2>, angularTarget=<0 0 0>}
      localJointOffset=<8 1 2>
     }]
    relativeSpaceBone {boneName=upperarm_r}
   }
   n11 = Anim Dynamics - Bone: dyn_bk_ArmBagginess_r(ComponentPose=@n10, Alpha=1, GravityScale=1, ExternalForce="0,0,0") {
    angularDampingOverride=0.9999
    angularSpringConstant=500
    bGravityOverrideInSimSpace=true
    bOverrideAngularDamping=true
    bUseGravityOverride=true
    simulationSpace=RootRelative
    alphaBoolBlend {blendOption=Linear}
    boundBone {boneName=dyn_bk_ArmBagginess_r}
    componentAppliedLinearAccClamp=<3000 3000 3000>
    componentLinearAccScale=<0.1 0.1 0.1>
    componentLinearVelScale=<0.3 0.3 0.3>
    gravityOverride {z=-500}
    physicsBodyDefinitions [{
      sphereCollisionRadius=0
      boundBone {boneName=dyn_bk_ArmBagginess_r}
      boxExtents=<10 20 20>
      constraintSetup {angularLimitsMax=<3 3 2>, angularLimitsMin=<-3 -5 -2>, angularTarget=<0 0 0>}
      localJointOffset=<8 -1 2>
     }]
    relativeSpaceBone {boneName=upperarm_r}
   }
   n12 = Anim Dynamics - Bone: dyn_up_ArmBagginess_r(ComponentPose=@n11, Alpha=1, GravityScale=1, ExternalForce="0,0,0") {
    angularDampingOverride=1
    angularSpringConstant=500
    bGravityOverrideInSimSpace=true
    bOverrideAngularDamping=true
    bUseGravityOverride=true
    simulationSpace=BoneRelative
    alphaBoolBlend {blendOption=Linear}
    boundBone {boneName=dyn_up_ArmBagginess_r}
    componentAppliedLinearAccClamp=<3000 3000 3000>
    componentLinearAccScale=<0.1 0.1 0.1>
    componentLinearVelScale=<0.2 0.2 0.2>
    gravityOverride {y=500, z=50}
    physicsBodyDefinitions [{
      sphereCollisionRadius=0
      boundBone {boneName=dyn_up_ArmBagginess_r}
      boxExtents=<15 20 20>
      constraintSetup {angularLimitsMax=<0 2 0>, angularLimitsMin=<0 -5 -5>, angularTarget=<0 0 0>}
      localJointOffset=<8 0 0>
     }]
    relativeSpaceBone {boneName=upperarm_twist_01_r}
   }
   n13 = Anim Dynamics - Bone: dyn_lw_ArmBagginess_l(ComponentPose=@n12, Alpha=1, GravityScale=1, ExternalForce="0,0,0") {
    angularDampingOverride=0.99
    angularSpringConstant=1000
    bAngularSpring=true
    bGravityOverrideInSimSpace=true
    bOverrideAngularDamping=true
    bUseGravityOverride=true
    simulationSpace=RootRelative
    alphaBoolBlend {blendOption=Linear}
    boundBone {boneName=dyn_lw_ArmBagginess_l}
    componentAppliedLinearAccClamp=<3000 3000 3000>
    componentLinearAccScale=<0.2 0.2 0.2>
    componentLinearVelScale=<0.3 0.3 0.3>
    gravityOverride {z=-500}
    physicsBodyDefinitions [{
      sphereCollisionRadius=0
      boundBone {boneName=dyn_lw_ArmBagginess_l}
      boxExtents=<20 20 10>
      constraintSetup {angularLimitsMax=<3 2 10>, angularLimitsMin=<-3 -15 -10>, angularTarget=<0 0 0>}
      localJointOffset=<8 0 0>
     }]
    relativeSpaceBone {boneName=lowerarm_r}
   }
   n14 = Anim Dynamics - Bone: dyn_fr_ArmBagginess_l(ComponentPose=@n13, Alpha=1, GravityScale=1, ExternalForce="0,0,0") {
    angularDampingOverride=0.9999
    angularSpringConstant=500
    bGravityOverrideInSimSpace=true
    bOverrideAngularDamping=true
    bUseGravityOverride=true
    simulationSpace=RootRelative
    alphaBoolBlend {blendOption=Linear}
    boundBone {boneName=dyn_fr_ArmBagginess_l}
    componentAppliedLinearAccClamp=<3000 3000 3000>
    componentLinearAccScale=<0.1 0.1 0.1>
    componentLinearVelScale=<0.3 0.3 0.3>
    gravityOverride {z=-500}
    physicsBodyDefinitions [{
      sphereCollisionRadius=0
      boundBone {boneName=dyn_fr_ArmBagginess_l}
      boxExtents=<10 20 20>
      constraintSetup {angularLimitsMax=<3 5 2>, angularLimitsMin=<-3 -5 -2>, angularTarget=<0 0 0>}
      localJointOffset=<8 1 2>
     }]
    relativeSpaceBone {boneName=upperarm_r}
   }
   n15 = Anim Dynamics - Bone: dyn_bk_ArmBagginess_l(ComponentPose=@n14, Alpha=1, GravityScale=1, ExternalForce="0,0,0") {
    angularDampingOverride=0.9999
    angularSpringConstant=500
    bGravityOverrideInSimSpace=true
    bOverrideAngularDamping=true
    bUseGravityOverride=true
    simulationSpace=RootRelative
    alphaBoolBlend {blendOption=Linear}
    boundBone {boneName=dyn_bk_ArmBagginess_l}
    componentAppliedLinearAccClamp=<3000 3000 3000>
    componentLinearAccScale=<0.1 0.1 0.1>
    componentLinearVelScale=<0.3 0.3 0.3>
    gravityOverride {z=-500}
    physicsBodyDefinitions [{
      sphereCollisionRadius=0
      boundBone {boneName=dyn_bk_ArmBagginess_l}
      boxExtents=<10 20 20>
      constraintSetup {angularLimitsMax=<3 5 2>, angularLimitsMin=<-3 -5 -2>, angularTarget=<0 0 0>}
      localJointOffset=<8 -1 2>
     }]
    relativeSpaceBone {boneName=upperarm_r}
   }
   n16 = Anim Dynamics - Bone: dyn_up_ArmBagginess_l(ComponentPose=@n15, Alpha=1, GravityScale=1, ExternalForce="0,0,0") {
    angularDampingOverride=1
    angularSpringConstant=500
    bGravityOverrideInSimSpace=true
    bOverrideAngularDamping=true
    bUseGravityOverride=true
    simulationSpace=BoneRelative
    alphaBoolBlend {blendOption=Linear}
    boundBone {boneName=dyn_up_ArmBagginess_l}
    componentAppliedLinearAccClamp=<3000 3000 3000>
    componentLinearAccScale=<0.1 0.1 0.1>
    componentLinearVelScale=<0.2 0.2 0.2>
    gravityOverride {y=500, z=-50}
    physicsBodyDefinitions [{
      sphereCollisionRadius=0
      boundBone {boneName=dyn_up_ArmBagginess_l}
      boxExtents=<15 20 20>
      constraintSetup {angularLimitsMax=<0 2 5>, angularLimitsMin=<0 -5 -5>, angularTarget=<0 0 0>}
      localJointOffset=<8 0 0>
     }]
    relativeSpaceBone {boneName=upperarm_twist_01_l}
   }
   Output Pose(Result=Blend Poses by bool(
     BlendPose_0=Component To Local(ComponentPose=@n8)
     BlendPose_1=Component To Local(ComponentPose=@n16)
     bActiveValue=Get FE_Locker_Idle() -> (FE_Locker_Idle:bool)
    ) {blendType=Linear}) {layerGroup=None}
   n17 = Control Rig!(
    Source=Copy Pose From Mesh() {bUseAttachedParent=true}
    Alpha
    CrouchingBool:bool=Get bIsCrouching()
    Moving2dBool:bool=Get bIsMoving2D()
    OpenGlider:bool=Get HangGlidingAlpha() -> (HangGlidingAlpha:bool)
    Pistol:double=Get Pistol() -> (Pistol:double)
    Emote:double=Get AlphaEmote() -> (AlphaEmote:double)
    Move_Back:double=Get Move_Backward() -> (Move_Backward:double)
    TacticalSprintingBool:bool=Get bIsTacticalSprinting()
    DBNO:bool=Get bIsDBNO()
    FrontEndBool:bool=Get bIsInFrontEnd()
    FE_Locker_Idle:bool=Get FE_Locker_Idle()
   ) -> (Pose) {
    alphaBoolBlend {blendOption=Linear}
    controlRigAssetReference {blueprintRigClass=&/M_MED_FluteLamp_CtrlRig.M_MED_FluteLamp_CtrlRig_C}
   }
   Save cached pose 'ControlRigCache'(Pose=@n17) {cacheName=ControlRigCache}
   n18 = RigidBody!(
    ComponentPose=Local To Component(LocalPose=Use cached pose 'ControlRigCache'())
    Alpha=0.7
    OverrideWorldGravity:Vector=Get Gravity Override()
    ExternalForce
    SimSpaceSettings:Sim Space Settings Structure=(WorldAlpha=0.5,VelocityScaleZ=0.5,DampingAlpha=1,MaxLinearVelocity=500,MaxAngularVelocity=10,MaxLinearAcceleration=300,MaxAngularAcceleration=500,ExternalLinearDragV=(X=0,Y=0,Z=0),ExternalLinearVelocity=(X=0,Y=0,Z=0),ExternalAngularVelocity=(X=0,Y=0,Z=0))
   ) -> (Pose) {
    bOverrideWorldGravity=true
    overridePhysicsAsset=&/M_MED_FluteLamp_Physics
    simulationSpace=BaseBoneSpace
    alphaBoolBlend {blendOption=Linear}
    baseBoneRef {boneName=pelvis}
    componentAppliedLinearAccClamp=<3000 3000 3000>
    simSpaceSettings {maxAngularAcceleration=500, maxAngularVelocity=10, maxLinearAcceleration=300, maxLinearVelocity=500, velocityScaleZ=0.5, worldAlpha=0.5}
   }
   n19 = Anim Dynamics - Bone: dyn_fr_ShirtBagginess_l_01!(ComponentPose=@n18, Alpha=1, GravityOverride:Vector="0,0,-980", ExternalForce="0,0,0") -> (Pose) {
    angularDampingOverride=0.9999
    angularSpringConstant=20000
    bAngularSpring=true
    bOverrideAngularDamping=true
    bUseGravityOverride=true
    gravityScale=0.35
    simulationSpace=BoneRelative
    alphaBoolBlend {blendOption=Linear}
    alphaScaleBiasClamp {bInterpResult=true}
    boundBone {boneName=dyn_fr_ShirtBagginess_l_01}
    componentAppliedLinearAccClamp=<2000 2000 2000>
    componentLinearAccScale=<1 1 1>
    componentLinearVelScale=<1 1 1>
    gravityOverride {z=-980}
    physicsBodyDefinitions [{
      boundBone {boneName=dyn_fr_ShirtBagginess_l_01}
      boxExtents=<20 30 30>
      constraintSetup {angularLimitsMax=<0 0 5>, angularLimitsMin=<0 -20 -5>, angularTarget=<1 0 0>}
      localJointOffset=<10 0 0>
     }]
    relativeSpaceBone {boneName=base_fr_ShirtBagginess_l}
   }
   n20 = Anim Dynamics - Bone: dyn_fr_ShirtBagginess_r_01!(ComponentPose=@n19, Alpha=1, GravityOverride:Vector="0,0,-980", ExternalForce="0,0,0") -> (Pose) {
    angularDampingOverride=0.9999
    angularSpringConstant=15000
    bAngularSpring=true
    bOverrideAngularDamping=true
    bUseGravityOverride=true
    gravityScale=0.35
    simulationSpace=BoneRelative
    alphaBoolBlend {blendOption=Linear}
    alphaScaleBiasClamp {bInterpResult=true}
    boundBone {boneName=dyn_fr_ShirtBagginess_r_01}
    componentAppliedLinearAccClamp=<2000 2000 2000>
    componentLinearAccScale=<0.3 0.3 0.3>
    componentLinearVelScale=<0.3 0.3 0.3>
    gravityOverride {z=-980}
    physicsBodyDefinitions [{
      boundBone {boneName=dyn_fr_ShirtBagginess_r_01}
      boxExtents=<20 30 30>
      constraintSetup {angularLimitsMax=<0 0 5>, angularLimitsMin=<0 -20 -5>, angularTarget=<1 0 0>}
      localJointOffset=<-10 0 0>
     }]
    relativeSpaceBone {boneName=base_fr_ShirtBagginess_r}
   }
   n21 = Anim Dynamics - Bone: dyn_fr_PantBagginess_l!(ComponentPose=@n20, Alpha=1, GravityOverride:Vector="0,0,-980", ExternalForce="0,0,0") -> (Pose) {
    angularDampingOverride=1
    angularSpringConstant=15000
    bAngularSpring=true
    bOverrideAngularDamping=true
    bUseGravityOverride=true
    gravityScale=0.35
    simulationSpace=BoneRelative
    alphaBoolBlend {blendOption=Linear}
    boundBone {boneName=dyn_fr_PantBagginess_l}
    componentAppliedLinearAccClamp=<2000 2000 2000>
    componentLinearAccScale=<1 1 1>
    componentLinearVelScale=<1 1 1>
    gravityOverride {z=-980}
    physicsBodyDefinitions [{
      boundBone {boneName=dyn_fr_PantBagginess_l}
      boxExtents=<20 30 30>
      constraintSetup {angularLimitsMax=<10 30 35>, angularLimitsMin=<-10 -5 -35>, angularTarget=<1 0 0>}
      localJointOffset=<10 0 0>
     }]
    relativeSpaceBone {boneName=root}
   }
   n22 = Anim Dynamics - Bone: dyn_bk_PantBagginess_l!(ComponentPose=@n21, Alpha=1, GravityOverride:Vector="0,0,-980", ExternalForce="0,0,0") -> (Pose) {
    angularDampingOverride=1
    angularSpringConstant=15000
    bAngularSpring=true
    bOverrideAngularDamping=true
    bUseGravityOverride=true
    gravityScale=0.35
    simulationSpace=BoneRelative
    alphaBoolBlend {blendOption=Linear}
    boundBone {boneName=dyn_bk_PantBagginess_l}
    componentAppliedLinearAccClamp=<2000 2000 2000>
    componentLinearAccScale=<1 1 1>
    componentLinearVelScale=<1 1 1>
    gravityOverride {z=-980}
    physicsBodyDefinitions [{
      boundBone {boneName=dyn_bk_PantBagginess_l}
      boxExtents=<20 30 30>
      constraintSetup {angularLimitsMax=<10 3 35>, angularLimitsMin=<-10 -15 -35>, angularTarget=<1 0 0>}
      localJointOffset=<10 0 0>
     }]
    relativeSpaceBone {boneName=root}
   }
   n23 = Anim Dynamics - Bone: dyn_fr_PantBagginess_r!(ComponentPose=@n22, Alpha=1, GravityOverride:Vector="0,0,-980", ExternalForce="0,0,0") -> (Pose) {
    angularDampingOverride=1
    angularSpringConstant=15000
    bAngularSpring=true
    bOverrideAngularDamping=true
    bUseGravityOverride=true
    gravityScale=0.35
    simulationSpace=BoneRelative
    alphaBoolBlend {blendOption=Linear}
    boundBone {boneName=dyn_fr_PantBagginess_r}
    componentAppliedLinearAccClamp=<2000 2000 2000>
    componentLinearAccScale=<1 1 1>
    componentLinearVelScale=<1 1 1>
    gravityOverride {z=-980}
    physicsBodyDefinitions [{
      boundBone {boneName=dyn_fr_PantBagginess_r}
      boxExtents=<20 30 30>
      constraintSetup {angularLimitsMax=<10 30 35>, angularLimitsMin=<-10 -5 -35>, angularTarget=<1 0 0>}
      localJointOffset=<-10 0 0>
     }]
    relativeSpaceBone {boneName=root}
   }
   n24 = Anim Dynamics - Bone: dyn_bk_PantBagginess_r!(ComponentPose=@n23, Alpha=1, GravityOverride:Vector="0,0,-980", ExternalForce="0,0,0") -> (Pose) {
    angularDampingOverride=1
    angularSpringConstant=15000
    bAngularSpring=true
    bOverrideAngularDamping=true
    bUseGravityOverride=true
    gravityScale=0.35
    simulationSpace=BoneRelative
    alphaBoolBlend {blendOption=Linear}
    boundBone {boneName=dyn_bk_PantBagginess_r}
    componentAppliedLinearAccClamp=<2000 2000 2000>
    componentLinearAccScale=<1 1 1>
    componentLinearVelScale=<1 1 1>
    gravityOverride {z=-980}
    physicsBodyDefinitions [{
      boundBone {boneName=dyn_bk_PantBagginess_r}
      boxExtents=<20 30 30>
      constraintSetup {angularLimitsMax=<10 30 35>, angularLimitsMin=<-10 -5 -35>, angularTarget=<1 0 0>}
      localJointOffset=<-10 0 0>
     }]
    relativeSpaceBone {boneName=root}
   }
   Save cached pose 'AnimDynamicsCache'(Pose=Component To Local(ComponentPose=@n24)) {cacheName=AnimDynamicsCache}
   n25 = "Transform (Modify) Bone - Bone: base_sd_ShirtBagginess_r"!(
    ComponentPose=Local To Component(LocalPose=Use cached pose 'AnimDynamicsCache'())
    bAlphaBoolEnabled:bool=True
    Translation="0,0,0"
    Rotation="4.860,0.000,-0.000"
    Scale="1,1,1"
   ) -> (Pose) {
    alphaInputType=Bool
    rotationMode=BMM_Additive
    rotationSpace=BCS_BoneSpace
    scaleMode=BMM_Additive
    scaleSpace=BCS_BoneSpace
    translationMode=BMM_Additive
    translationSpace=BCS_BoneSpace
    alphaBoolBlend {blendInTime=0.2, blendOption=Linear, blendOutTime=2}
    boneToModify {boneName=base_sd_ShirtBagginess_r}
    rotation=<rot 4.86 0 0>
   }
   n26 = "Transform (Modify) Bone - Bone: dyn_fr_ShirtBagginess_r_02"!(ComponentPose=@n25, bAlphaBoolEnabled:bool=True, Translation="2.601,0.810,-1.859", Rotation="5.949,19.180,-4.362", Scale="1,1,1") -> (Pose) {
    alphaInputType=Bool
    rotationMode=BMM_Additive
    rotationSpace=BCS_BoneSpace
    scaleMode=BMM_Additive
    scaleSpace=BCS_BoneSpace
    translationMode=BMM_Additive
    translationSpace=BCS_BoneSpace
    alphaBoolBlend {blendInTime=0.2, blendOption=Linear, blendOutTime=2}
    boneToModify {boneName=dyn_fr_ShirtBagginess_r_02}
    rotation=<rot 5.949 19.18 -4.362>
    translation=<2.6013 0.8102 -1.859>
   }
   n27 = "Transform (Modify) Bone - Bone: dyn_fr_ShirtBagginess_c_02"!(ComponentPose=@n26, bAlphaBoolEnabled:bool=True, Translation="-2.752,-0.923,-0.158", Rotation="-4.596,5.040,6.365", Scale="1,1,1") -> (Pose) {
    alphaInputType=Bool
    rotationMode=BMM_Additive
    rotationSpace=BCS_BoneSpace
    scaleMode=BMM_Additive
    scaleSpace=BCS_BoneSpace
    translationMode=BMM_Additive
    translationSpace=BCS_BoneSpace
    alphaBoolBlend {blendInTime=0.2, blendOption=Linear, blendOutTime=2}
    boneToModify {boneName=dyn_fr_ShirtBagginess_c_02}
    rotation=<rot -4.596 5.04 6.365>
    translation=<-2.752 -0.923 -0.158>
   }
   n28 = "Transform (Modify) Bone - Bone: base_fr_ShirtBagginess_l"!(ComponentPose=@n27, bAlphaBoolEnabled:bool=True, Translation="0,0,0", Rotation="5.163,-0.407,-10.522", Scale="1,1,1") -> (Pose) {
    alphaInputType=Bool
    rotationMode=BMM_Additive
    rotationSpace=BCS_BoneSpace
    scaleMode=BMM_Additive
    scaleSpace=BCS_BoneSpace
    translationMode=BMM_Additive
    translationSpace=BCS_BoneSpace
    alphaBoolBlend {blendInTime=0.2, blendOption=Linear, blendOutTime=2}
    boneToModify {boneName=base_fr_ShirtBagginess_l}
    rotation=<rot 5.1631 -0.407 -10.522>
   }
   n29 = "Transform (Modify) Bone - Bone: base_fr_ShirtBagginess_c"!(ComponentPose=@n28, bAlphaBoolEnabled:bool=True, Translation="0,0,0", Rotation="-10.992,-1.288,-0.473", Scale="1,1,1") -> (Pose) {
    alphaInputType=Bool
    rotationMode=BMM_Additive
    rotationSpace=BCS_BoneSpace
    scaleMode=BMM_Additive
    scaleSpace=BCS_BoneSpace
    translationMode=BMM_Additive
    translationSpace=BCS_BoneSpace
    alphaBoolBlend {blendInTime=0.5, blendOption=CubicInOut, blendOutTime=0.2}
    boneToModify {boneName=base_fr_ShirtBagginess_c}
    rotation=<rot -10.992 -1.288 -0.473>
   }
   n30 = "Transform (Modify) Bone - Bone: base_sd_ShirtBagginess_r"(ComponentPose=@n29, bAlphaBoolEnabled=True, Translation="0,0,0", Rotation="20.318,0.567,16.672", Scale="1,1,1") {
    alphaInputType=Bool
    rotationMode=BMM_Additive
    rotationSpace=BCS_BoneSpace
    scaleMode=BMM_Additive
    scaleSpace=BCS_BoneSpace
    translationMode=BMM_Additive
    translationSpace=BCS_BoneSpace
    alphaBoolBlend {blendInTime=0.2, blendOption=Linear, blendOutTime=0.2}
    boneToModify {boneName=base_sd_ShirtBagginess_r}
    rotation=<rot 20.318 0.567 16.672>
   }
   n31 = "Transform (Modify) Bone - Bone: dyn_fr_ShirtBagginess_r_02"(ComponentPose=@n30, bAlphaBoolEnabled=True, Translation="0,0,0", Rotation="7.656,0.000,33.814", Scale="1,1,1") {
    alphaInputType=Bool
    rotationMode=BMM_Additive
    rotationSpace=BCS_BoneSpace
    scaleMode=BMM_Additive
    scaleSpace=BCS_BoneSpace
    translationMode=BMM_Additive
    translationSpace=BCS_BoneSpace
    alphaBoolBlend {blendInTime=0.2, blendOption=Linear, blendOutTime=0.2}
    boneToModify {boneName=dyn_fr_ShirtBagginess_r_02}
    rotation=<rot 7.656 0 33.814>
   }
   n32 = "Transform (Modify) Bone - Bone: dyn_ArmBagginessArea_l"!(ComponentPose=@n31, bAlphaBoolEnabled:bool=True, Translation="0,0,0", Rotation="11.278,-6.866,-0.000", Scale="1,1,1") -> (Pose) {
    alphaInputType=Bool
    rotationMode=BMM_Additive
    rotationSpace=BCS_BoneSpace
    scaleMode=BMM_Additive
    scaleSpace=BCS_BoneSpace
    translationMode=BMM_Additive
    translationSpace=BCS_BoneSpace
    alphaBoolBlend {blendInTime=0.2, blendOption=Linear, blendOutTime=0.2}
    boneToModify {boneName=dyn_ArmBagginessArea_l}
    rotation=<rot 11.278 -6.866 0>
   }
   n33 = "Transform (Modify) Bone - Bone: dyn_ArmBagginessArea_r"!(ComponentPose=@n32, bAlphaBoolEnabled:bool=True, Translation="0,0,0", Rotation="16.629,-14.508,-0.200", Scale="1,1,1") -> (Pose) {
    alphaInputType=Bool
    rotationMode=BMM_Additive
    rotationSpace=BCS_BoneSpace
    scaleMode=BMM_Additive
    scaleSpace=BCS_BoneSpace
    translationMode=BMM_Additive
    translationSpace=BCS_BoneSpace
    alphaBoolBlend {blendInTime=0.2, blendOption=Linear, blendOutTime=0.2}
    boneToModify {boneName=dyn_ArmBagginessArea_r}
    rotation=<rot 16.629 -14.508 -0.2>
   }
   Save cached pose 'FixedCache'(Pose=Component To Local(ComponentPose=@n33)) {cacheName=FixedCache}
  ]}
 EventGraph {nodes [
   Event BlueprintInitializeAnimation()
   > Set SkelMesh(SkelMesh=GetOwnerSkeletalMeshComponent())
   n1 = TryGetPawnOwner()
   > Is Valid(InputObject=@n1)
   > n2 = Cast To FortPlayerPawnAthena(Object=@n1)
   > n3 = Set PlayerPawn(PlayerPawn:Fort Player Pawn Athena Object Reference=@n2) -> (Output_Get:Fort Player Pawn Athena Object Reference)
   > Set Body_Mesh(Body_Mesh=GetSkeletalMeshForPartType(self=@n3))
   n4 = Event BlueprintUpdateAnimation()
   > n5 = Sequence!(execute) -> (then_0, then_1, then_2:Exec, then_3:Exec, then_4:Exec, then_5:Exec, then_6:Exec, then_7:Exec)
   n6 = "Gravity Override State Function (Parameter Asset) (Nativized)"(
    execute=@n5.then_1
    CharacterPartAnimInstance=Self-Reference()
    GameTimeInSeconds=GetGameTimeInSeconds()
    DebugJointName=spine_05
    Parameters=&/M_MED_FluteLamp_Parameters
   )
   > Set Gravity Override(Gravity Override=@n6)
   > Set RBAN Settings Out!(execute, RBAN Settings Out=@n6.OutRigidBodyAnimNodeParameters) -> (
    then
    Output_Get_ComponentLinearAccScale:Vector=~
    Output_Get_ComponentLinearVelScale:Vector=~
    Output_Get_SimSpaceSettings:Sim Space Settings Structure
   )
   n7 = PlayerMovement(execute=@n5.then_5, AnimInstance=Self-Reference(), XDotProductCoefficient=1, YDotProductCoefficient=1, ZDotProductCoefficient=1)
   n8 = BreakVector(InVec=MakeVector(X=@n7, Y=@n7.YMovement))
   > Set Move_Left(Move_Left:double="Absolute (Float)"(A="Clamp (Float)"(Value=@n8, Min=-1, Max=0))) -> (Output_Get:double)
   > Set Move_Right(Move_Right:double="Clamp (Float)"(Value=@n8, Min=0, Max=1)) -> (Output_Get:double)
   > Set Move_Forward(Move_Forward:double="Clamp (Float)"(Value=@n8.Y, Min=0, Max=1)) -> (Output_Get:double)
   > Set Move_Backward(Move_Backward:double="Absolute (Float)"(A="Clamp (Float)"(Value=@n8.Y, Min=-1, Max=0))) -> (Output_Get:double)
   n9 = Get bIsSkyDiving()
   Set AlphaNotSkydiving(execute=@n5.then_6, AlphaNotSkydiving:double="To Float (Boolean)"(InBool=NOT Boolean(A=@n9))) -> (Output_Get:double)
   n10 = NOT Boolean(A=Get bIsPlayingEmote())
   n11 = NOT Boolean(A=Get bIsSkyDiving())
   n12 = NOT Boolean(A=Get bIsSkydiveDiveMode())
   Set FE_Locker_Idle(execute=@n5.then_7, FE_Locker_Idle:bool=AND Boolean(A=AND Boolean!(A=@n11, B=@n10, C:bool=@n12) -> (ReturnValue), B=Get bIsInFrontEnd())) -> (Output_Get:bool)
   n13 = Get bIsSkyDiving()
   n14 = Branch(execute=@n5.then_4, Condition=AND Boolean!(A=Get SelectedPistolBool() -> (SelectedPistolBool:bool), B=NOT Boolean(A=@n13)) -> (ReturnValue))
   > Set Pistol(Pistol:double=1) -> (Output_Get:double)
   Set Pistol(execute=@n14.else, Pistol=0)
   Sequence!(execute=@n5.then_0) -> (then_0, then_1)
   > Is Valid(InputObject=Get PlayerPawn() -> (PlayerPawn:Fort Player Pawn Athena Object Reference))
   > n15 = Branch(Condition=Get bIsInFrontEnd())
   Set WPO Globals Function(
    execute=@n15.else
    Material=Get Mat Instance Dynamic()
    Overlay Material=Get Overlay Mat Instance Dynamic()
    Skeletal Mesh=Get SkelMesh()
    "Local Noise Transform  Location"=Get Local Noise Transform Location() -> (Local Noise Transform Location:Vector)
    Local Noise Offset=Get Local Noise Offset()
   )
   > Set WPO Regional Function!(
    execute
    Material:Material Instance Dynamic Object Reference=Get Mat Instance Dynamic()
    Overlay Material:Material Instance Dynamic Object Reference=Get Overlay Mat Instance Dynamic()
    Skeletal Mesh:Skeletal Mesh Component Object Reference=Get SkelMesh()
    Velocity Transform:Name=foot_r
    Velocity Interp Speed:double=Get Velocity Interp Speed()
    Collision Transform:Name=calf_r
    Collision Transform Axis:EAxis Enum
    Invert Collision Transform Axis:bool=false
    Delta Time X:double=@n4.DeltaTimeX
    Previous Velocity Transform Location:"Vector (by ref)"=Get R1 Velocity Transform Location()
    Previous Velocity Transform Velocity:"Vector (by ref)"=Get R1 Velocity Transform Velocity()
    Region:int=1
   ) -> (then)
   > Set WPO Regional Function(
    Material=Get Mat Instance Dynamic()
    Overlay Material=Get Overlay Mat Instance Dynamic()
    Skeletal Mesh=Get SkelMesh()
    Velocity Transform=lowerarm_r
    Velocity Interp Speed=Get Velocity Interp Speed()
    Collision Transform=upperarm_r
    Invert Collision Transform Axis=true
    Delta Time X=@n4.DeltaTimeX
    Previous Velocity Transform Location=Get R2 Velocity Transform Location()
    Previous Velocity Transform Velocity=Get R2 Velocity Transform Velocity()
    Region=2
   )
   > n16 = Branch(Condition=OR Boolean(A=Get bIsSkydiveFloating(), B=Get bIsParachuteOpened()))
   > Set WPO Pho Wind Intensity(WPO Pho Wind Intensity=0.55)
   > n17 = Set WPO Pho Turbulence(WPO Pho Turbulence:double=1.5) -> (Output_Get:double)
   SetScalarParameterValue(
    execute=[@n19.then, @n17.then]
    self=[Get Mat Instance Dynamic(), Get Overlay Mat Instance Dynamic()]
    ParameterName=ClothSim_WindStrength
    Value=Get WPO Pho Wind Intensity()
   )
   n18 = Get WPO Pho Turbulence() -> (WPO Pho Turbulence:double)
   > SetScalarParameterValue(self=Get Mat Instance Dynamic(), ParameterName=ClothSim_R1_TurbulenceStrength, Value=@n18)
   > SetScalarParameterValue(self=Get Mat Instance Dynamic(), ParameterName=ClothSim_R2_TurbulenceStrength, Value=@n18)
   Set WPO Pho Wind Intensity(execute=@n16.else, WPO Pho Wind Intensity=0)
   > n19 = Set WPO Pho Turbulence(WPO Pho Turbulence=0)
   n20 = Branch(execute=@n5.then_3, Condition=Get bIsPlayingEmote())
   > Set AlphaEmote(AlphaEmote:double=0.85) -> (Output_Get:double)
   Set AlphaEmote(execute=@n20.else, AlphaEmote=1)
   n21 = Branch(execute=@n5.then_2, Condition=Get bIsSkyDiving())
   > Set AlphaDynSkydiving(AlphaDynSkydiving:double=1) -> (Output_Get:double)
   > Set AlphaDynOnground(AlphaDynOnground:double=0) -> (Output_Get:double)
   Set AlphaDynSkydiving(execute=@n21.else, AlphaDynSkydiving=0)
   > Set AlphaDynOnground(AlphaDynOnground=1)
   Event OnParachuteOpened()
   > Switch on EGliderType(Selection=Get PlayerGliderType() -> (PlayerGliderType=HangGlider))
   > Set HangGlidingAlpha(HangGlidingAlpha:bool=1) -> (Output_Get:bool)
   Event OnParachuteClosed()
   > Set HangGlidingAlpha(HangGlidingAlpha=false)
   Event OnWeaponChanged()
   > n22 = Cast To FortPlayerAnimInstance(Object=GetAnimInstance(self=GetOwnerSkeletalMeshComponent()))
   > Set SelectedPistolBool(SelectedPistolBool:bool=false) -> (Output_Get:bool)
   > n23 = Switch on EFortWeaponCoreAnimation!(execute, Selection:EFortWeaponCoreAnimation Enum=Get WeaponCoreAnim(self=@n22)) -> (
    Melee:Exec
    Pistol:Exec
    Shotgun:Exec
    PaperBlueprint:Exec
    Rifle:Exec
    MeleeOneHand:Exec
    MachinePistol:Exec
    RocketLauncher:Exec
    GrenadeLauncher:Exec
    GoingCommando:Exec
    AssaultRifle:Exec
    TacticalShotgun:Exec
    SniperRifle:Exec
    TrapPlacement:Exec
    ShoulderLauncher:Exec
    AbilityDecoTool:Exec
    Crossbow:Exec
    C4:Exec
    RemoteControl:Exec
    DualWield:Exec
    AR_BullPup:Exec
    AR_ForwardGrip:Exec
    MedPackPaddles:Exec
    SMG_P90:Exec
    AR_DrumGun:Exec
    Consumable_Small:Exec
    Consumable_Large:Exec
    Balloon:Exec
    MountedTurret:Exec
    CreativeTool:Exec
    ExplosiveBow:Exec
    AshtonIndigo:Exec
    AshtonChicago:Exec
    MeleeDualWield:Exec
    Unarmed:Exec
   )
   Set SelectedPistolBool(execute=@n23.Pistol, SelectedPistolBool=true)
   Event OnCharacterCustomizationLoaded()
   > n24 = Set SkelMesh(SkelMesh=GetPartSkeletalMeshComponent())
   > n25 = Get Dynamic Material Instance(SkeletalMeshComponent=@n24)
   > Set Mat Instance Dynamic(Mat Instance Dynamic=@n25)
   > n26 = GetOrCreateDynamicOverlayMaterial(MeshComponent=@n24, MaterialName=None)
   > Set Overlay Mat Instance Dynamic(Overlay Mat Instance Dynamic=@n26)
  ]}
]
variables [
 Alpha_MeleeArm L {type=number}
 Alpha_MeleeArm R {type=number}
 AlphaDynOnground {type=number}
 AlphaDynSkydiving {type=number}
 AlphaEmote {type=number}
 AlphaNotSkydiving {type=number}
 Body_Mesh {type=SkeletalMeshComponent}
 FE_Locker_Idle {type=boolean}
 Gravity Override {type=Vector}
 HangGlidingAlpha {type=boolean}
 Local Noise Offset {type=Vector}
 Local Noise Transform Location {type=Vector}
 Mat Instance Dynamic {type=MaterialInstanceDynamic}
 Move_Backward {type=number}
 Move_Forward {type=number}
 Move_Left {type=number}
 Move_Right {type=number}
 Overlay Mat Instance Dynamic {type=MaterialInstanceDynamic}
 Pistol {type=number}
 PlayerPawn {type=FortPlayerPawnAthena}
 R1 Velocity Transform Location {type=Vector}
 R1 Velocity Transform Velocity {type=Vector}
 R2 Velocity Transform Location {type=Vector}
 R2 Velocity Transform Velocity {type=Vector}
 RBAN Settings Out {type=FortRigidBodyAnimNodeParameters}
 SelectedPistolBool {type=boolean}
 SkelMesh {type=SkeletalMeshComponent}
 Velocity Interp Speed {type=number}
 WPO Pho Turbulence {type=number}
 WPO Pho Wind Intensity {type=number}
]
