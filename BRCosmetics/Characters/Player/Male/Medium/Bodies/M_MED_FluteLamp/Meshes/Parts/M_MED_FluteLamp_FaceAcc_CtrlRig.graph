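# Control Rig for the M_MED_FluteLamp face accessory that applies corrective offsets to dyn_hair bones, including dyn_hair_11 and 12, as the head moves.
# A spherical pose reader measures head orientation against spine_03.
# Its output is multiplied by CrouchingBool and StandingBool, converted to floats, to weight two stance-specific Modify Transforms nodes.
# The alpha interp nodes have interpolation off, and FloatingBool is declared but unused.
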ControlRigBlueprint /BRCosmetics/Characters/Player/Male/Medium/Bodies/M_MED_FluteLamp/Meshes/Parts/M_MED_FluteLamp_FaceAcc_CtrlRig
graphs [
 RigVMFunctionLibrary {links [], nodes []}
 RigVMModel {nodes [
   RigUnit_BeginExecution = &RigUnit_BeginExecution()
   n1 = &RigUnit_SphericalPoseReader(
    ActiveRegionScaleFactors=(PositiveWidth=3,NegativeWidth=0.3,PositiveHeight=5,NegativeHeight=5)
    ActiveRegionSize=0.1
    Debug=(bDrawDebug=false,DebugScale=25,DebugSegments=20,DebugThickness=0.25)
    DriverAxis=(X=1,Y=0,Z=0)
    DriverItem=(Type=Bone,Name="head")
    FalloffRegionScaleFactors=(PositiveWidth=-4,NegativeWidth=1.2,PositiveHeight=-1.5,NegativeHeight=-1.5)
    FalloffSize=0.25
    OptionalParentItem=(Type=Bone,Name="spine_03")
    RotationOffset=(X=90,Y=0,Z=-140)
    ExecutePin=@RigUnit_BeginExecution.ExecutePin
   )
   RigVMFunction_MathBoolToFloat = &RigVMFunction_MathBoolToFloat(Value=RigVMVariableNode() {variable=CrouchingBool}.Value)
   AlphaInterp = &RigVMFunction_AlphaInterp(
    InterpSpeedDecreasing=10
    InterpSpeedIncreasing=10
    Scale=1
    bClampResult=False
    bInterpResult=False
    Value=@RigVMFunction_MathBoolToFloat.Result
   )
   Multiply = &RigVMFunction_MathDoubleMul(B=@AlphaInterp.Result, A=@n1.OutputParam)
   > &RigUnit_ModifyTransforms(
    ItemToModify="((Item=(Type=Bone,Name=\"dyn_hair_11\"),Transform=(Rotation=(X=0,Y=0,Z=0.3355,W=-0.942),Translation=(X=-0.1601,Y=-0.1963,Z=0),Scale3D=(X=1,Y=1,Z=1))),(Item=(Type=Bone,Name=\"dyn_hair_12\"),Transform=(Rotation=(X=0,Y=0,Z=0.3679,W=-0.9298),Translation=(X=-0.3698 ...(truncated)"
    WeightMinimum=0
    Weight=@Multiply.Result
   )
   n2 = &RigVMFunction_AlphaInterp(
    InterpSpeedDecreasing=10
    InterpSpeedIncreasing=10
    Scale=1
    bClampResult=False
    bInterpResult=False
    Value=&RigVMFunction_MathBoolToFloat(Value=RigVMVariableNode() {variable=StandingBool}.Value).Result
   )
   > &RigUnit_ModifyTransforms(
    ItemToModify="((Item=(Type=Bone,Name=\"dyn_hair_11\"),Transform=(Rotation=(X=0,Y=0,Z=0.0942,W=-0.9956),Translation=(X=-0.0826,Y=-0.4325,Z=0),Scale3D=(X=1,Y=1,Z=1))),(Item=(Type=Bone,Name=\"dyn_hair_12\"),Transform=(Rotation=(X=0,Y=0,Z=0.3155,W=-0.9489),Translation=(X=0.2,Y= ...(truncated)"
    WeightMinimum=0
    Weight=&RigVMFunction_MathDoubleMul(B=@n2.Result, A=@n1.OutputParam).Result
   )
  ]}
]
hierarchy [
 root {loc=<0 0 0>, rot=<0 0 0 1>}
 pelvis {parent=root, loc=<0 -0.7071 93.1191>, rot=<0.0127 0.707 -0.0127 -0.707>}
 spine_01 {parent=pelvis, loc=<6.7494 -0.6388 0>, rot=<0 0 0.034 -0.9994>}
 spine_02 {parent=spine_01, loc=<7.2518 -0.035 0>, rot=<0 0 0.0238 -0.9997>}
 spine_03 {parent=spine_02, loc=<6.0104 0 0>, rot=<0 0 0.0149 -0.9999>}
 spine_04 {parent=spine_03, loc=<8.1024 0 0>, rot=<0 0 -0.0132 -0.9999>}
 spine_05 {parent=spine_04, loc=<10.0146 0 0>, rot=<0 0 -0.0015 -1>}
 neck_01 {parent=spine_05, loc=<11.3138 0.1529 0>, rot=<0 0 -0.2003 -0.9797>}
 neck_02 {parent=neck_01, loc=<4.5627 0 0>, rot=<0 0 0.0504 -0.9987>}
 head {parent=neck_02, loc=<5.3914 0 0>, rot=<0 0 0.115 -0.9934>}
 dyn_hair_07 {parent=head, loc=<10.717 4.2117 -7.2497>, rot=<0.054 0.6246 0.7789 -0.0137>}
 dyn_hair_08 {parent=head, loc=<10.0576 2.5332 -7.7955>, rot=<0.0533 0.6603 0.7489 -0.0162>}
 dyn_hair_09 {parent=head, loc=<11.5215 1.4444 -8.7457>, rot=<-0.0614 0.6501 0.7483 -0.1167>}
 dyn_hair_10 {parent=head, loc=<9.3899 -3.2822 -9.5519>, rot=<0.0934 0.6396 0.7454 0.1633>}
 dyn_hair_11 {parent=head, loc=<1.3574 -3.4008 -3.6616>, rot=<0.0835 -0.9451 -0.3146 -0.0307>}
 dyn_hair_12 {parent=head, loc=<0.5486 -4.8176 -0.0146>, rot=<0.1081 -0.9939 0.0142 -0.0138>}
 dyn_hair_13 {parent=head, loc=<13.938 2.73 7.5587>, rot=<-0.0267 -0.5676 0.8226 0.0184>}
 dyn_hair_14 {parent=dyn_hair_13, loc=<3.1901 0 0>, rot=<0 0 0 1>}
 dyn_hair_15 {parent=head, loc=<11.9457 0.2618 8.9392>, rot=<-0.0516 -0.5659 0.8227 -0.0177>}
 dyn_hair_16 {parent=head, loc=<4.85 -2.0003 -8.4888>, rot=<-0.1947 -0.7422 -0.6366 -0.0781>}
 dyn_hair_17 {parent=head, loc=<3.3234 -1.3602 -6.3362>, rot=<-0.1947 -0.7422 -0.6366 -0.0781>}
 dyn_hair_18 {parent=head, loc=<12.9006 -8.4814 -8.0107>, rot=<0.136 -0.9132 -0.3778 -0.0694>}
 dyn_hair_19 {parent=head, loc=<4.3003 -1.2056 6.3686>, rot=<-0.185 -0.5612 0.8032 0.0747>}
 dyn_hair_20 {parent=head, loc=<9.4949 -3.9322 9.5697>, rot=<0.0574 -0.6109 0.7503 0.246>}
 dyn_hair_21 {parent=head, loc=<5.2671 -3.1472 8.9035>, rot=<-0.0657 -0.5946 0.7743 0.2064>}
 dyn_hair_22 {parent=head, loc=<1.3579 -3.4008 3.6616>, rot=<0.1269 -0.9427 0.3062 0.0381>}
 dyn_hair_23 {parent=head, loc=<6.334 -7.017 8.0846>, rot=<0.0104 -0.3886 0.9208 0.0318>}
 dyn_hair_24 {parent=head, loc=<19.2805 -4.1788 4.3388>, rot=<0.406 0.1732 0.3486 -0.8269>}
 dyn_hair_b_01 {parent=dyn_hair_24, loc=<2.906 -0.606 0>, rot=<0.0182 -0.003 0.2186 -0.9756>}
 dyn_hair_25 {parent=head, loc=<19.9432 -4.3625 3.8516>, rot=<-0.7671 -0.0208 0.1792 -0.6157>}
 dyn_hair_b_02 {parent=dyn_hair_25, loc=<2.1422 -0.3392 0>, rot=<0 0 0.1327 -0.9912>}
 dyn_hair_26 {parent=head, loc=<6.3782 -8.959 -6.2549>, rot=<-0.013 -0.9285 -0.365 0.0671>}
 dyn_hair_27 {parent=head, loc=<9.2475 -10.6426 -4.9393>, rot=<0.0492 -0.9724 -0.1817 0.1378>}
 dyn_hair_28 {parent=head, loc=<5.413 -5.8369 -8.7302>, rot=<-0.197 -0.8096 -0.5519 -0.0337>}
 dyn_hair_29 {parent=head, loc=<2.0548 -8.2671 -1.6974>, rot=<-0.0856 -0.9801 -0.1531 0.0934>}
 dyn_hair_30 {parent=head, loc=<2.6631 -8.5863 3.5708>, rot=<-0.1139 -0.9841 0.1358 -0.0128>}
 dyn_hair_31 {parent=head, loc=<9.1495 -8.7933 6.2837>, rot=<-0.0162 -0.367 0.9067 -0.2071>}
 dyn_hair_32 {parent=head, loc=<8.0913 -11.4207 0.5241>, rot=<0.007 -0.9771 0.104 0.1858>}
 faceAttach {parent=head, loc=<5.5687 -0.4551 -0.0668>, rot=<0.0029 -0.7071 -0.0029 -0.7071>}
 C_brow_mid {parent=faceAttach, loc=<-0.0641 10.2894 4.64>, rot=<0 0 0 -1>}
 L_brow_mid {parent=faceAttach, loc=<3.0017 9.636 5.7772>, rot=<-0.0001 -0.0042 0.2553 -0.9669>}
 L_brow_outer {parent=faceAttach, loc=<5.394 7.4112 4.6908>, rot=<-0.0287 -0.04 0.4491 -0.8921>}
 L_cheek_inner {parent=faceAttach, loc=<4.4419 8.5064 -0.7983>, rot=<-0.0706 -0.2697 0.3184 -0.906>}
 C_lip_upper_mid {parent=faceAttach, loc=<-0.0643 10.9851 -3.0362>, rot=<0.2225 0.0013 -0.0005 -0.9749>}
 L_lip_upper_outer {parent=faceAttach, loc=<1.3083 10.4253 -3.2668>, rot=<0.0359 -0.0728 0.2627 -0.9615>}
 L_lip_corner {parent=faceAttach, loc=<2.8018 9.1909 -3.2051>, rot=<0.003 0.008 -0.0146 -0.9999>}
 C_nose_bridge {parent=faceAttach, loc=<-0.0835 13.0147 0.5612>, rot=<-0.2218 0 0 -0.9751>}
 R_brow_mid {parent=faceAttach, loc=<-3.0535 9.6679 5.7622>, rot=<-0.0001 0.0042 -0.2553 -0.9669>}
 R_brow_outer {parent=faceAttach, loc=<-5.4094 7.6677 4.855>, rot=<-0.0287 0.04 -0.4491 -0.8921>}
 R_cheek_inner {parent=faceAttach, loc=<-4.5288 8.5511 -0.7504>, rot=<-0.0706 0.2697 -0.3184 -0.906>}
 R_lip_corner {parent=faceAttach, loc=<-2.7842 9.2382 -2.9981>, rot=<0.003 -0.008 0.0146 -0.9999>}
 C_jaw {parent=faceAttach, loc=<0.0387 1.3062 -0.8719>, rot=<0 0 0 1>}
 L_lip_lower_outer {parent=C_jaw, loc=<1.2553 9.0375 -2.3601>, rot=<0.013 0.0445 0.3057 -0.951>}
 C_lip_lower_mid {parent=C_jaw, loc=<-0.1173 9.976 -2.3476>, rot=<0 0 0 -1>}
 R_lip_lower_outer {parent=C_jaw, loc=<-1.7444 9.2084 -2.2009>, rot=<0.013 -0.0445 -0.3057 -0.951>}
 R_eye {parent=faceAttach, loc=<-2.7367 7.5891 3.9276>, rot=<0 0 0.9998 -0.0175>}
 R_eye_lid_upper_mid {parent=R_eye, loc=<0 0 0>, rot=<0.0523 0 0 -0.9986>}
 R_eye_lid_lower_mid {parent=R_eye, loc=<0 0 0>, rot=<0.0329 0 0 -0.9995>}
 L_eye {parent=faceAttach, loc=<2.6002 7.5885 3.9274>, rot=<0 0 0.9998 0.0175>}
 L_eye_lid_upper_mid {parent=L_eye, loc=<0 0 0>, rot=<0.0523 0 0 -0.9986>}
 L_eye_lid_lower_mid {parent=L_eye, loc=<0 0 0>, rot=<0.0329 0 0 -0.9995>}
 R_lip_upper_outer {parent=faceAttach, loc=<-1.4673 10.4099 -3.2344>, rot=<0.0359 0.0728 -0.2627 -0.9615>}
 dyn_hair_33 {parent=head, loc=<21.3762 9.1477 0.1584>, rot=<0.1702 -0.1038 -0.7973 -0.5697>}
 dyn_hair_01 {parent=dyn_hair_33, loc=<2.924 3.5277 3.3496>, rot=<0.2734 0.2704 -0.5153 -0.7659>}
 dyn_hair_02 {parent=dyn_hair_33, loc=<3.1999 0.6413 3.0408>, rot=<0.0574 0.0209 -0.2483 -0.9667>}
 dyn_hair_03 {parent=dyn_hair_33, loc=<4.555 3.9725 -0.0389>, rot=<0.0455 0.1136 -0.5708 -0.8119>}
 dyn_hair_04 {parent=dyn_hair_33, loc=<4.7437 4.6925 -3.1271>, rot=<-0.0562 0.0579 -0.5794 -0.8111>}
 dyn_hair_05 {parent=dyn_hair_33, loc=<5.8094 2.29 -2.6423>, rot=<-0.033 0.058 -0.5791 -0.8125>}
 dyn_hair_06 {parent=dyn_hair_33, loc=<3.8564 5.787 -6.0614>, rot=<-0.2822 -0.0344 -0.5861 -0.7588>}
]
variables [
 CrouchingBool {type=bool, default=False}
 FloatingBool {type=bool, default=False}
 StandingBool {type=bool, default=False}
]
