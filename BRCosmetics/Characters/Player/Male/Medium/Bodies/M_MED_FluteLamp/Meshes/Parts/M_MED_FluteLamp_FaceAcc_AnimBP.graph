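# Post-copy AnimBP for the M_MED_FluteLamp face accessory that adds secondary motion to its hair bones.
# The attached parent's pose feeds a Control Rig driven by crouch, skydive-float and standing flags, then a serial chain of per-bone AnimDynamics springs with angular limits.
# Fixed additive bone offsets adjust four hair bones after simulation.
# The event graph updates standing state and a neck-based head-rotation alpha, which the AnimGraph never reads.
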AnimBlueprint /BRCosmetics/Characters/Player/Male/Medium/Bodies/M_MED_FluteLamp/Meshes/Parts/M_MED_FluteLamp_FaceAcc_AnimBP
target_skeleton=&/M_MED_FluteLamp_FaceAcc_Skeleton
graphs [
 AnimGraph {nodes [
   n1 = "Transform (Modify) Bone - Bone: dyn_hair_11"(
    ComponentPose=Local To Component(LocalPose=Use cached pose 'DynCache'())
    Alpha=1
    Translation="0,0,0"
    Rotation="1.055,-15.684,10.048"
    Scale="1,1,1"
   ) {
    rotationMode=BMM_Additive
    rotationSpace=BCS_BoneSpace
    scaleMode=BMM_Additive
    scaleSpace=BCS_BoneSpace
    translationMode=BMM_Additive
    translationSpace=BCS_BoneSpace
    alphaBoolBlend {blendOption=Linear}
    alphaScaleBiasClamp {bInterpResult=true}
    boneToModify {boneName=dyn_hair_11}
    rotation=<rot 1.055 -15.684 10.048>
   }
   n2 = "Transform (Modify) Bone - Bone: dyn_hair_12"(ComponentPose=@n1, Alpha=1, Translation="0.064,-0.806,-0.000", Rotation="9.886,-38.922,0.146", Scale="1,1,1") {
    rotationMode=BMM_Additive
    rotationSpace=BCS_BoneSpace
    scaleMode=BMM_Additive
    scaleSpace=BCS_BoneSpace
    translationMode=BMM_Additive
    translationSpace=BCS_BoneSpace
    alphaBoolBlend {blendOption=Linear}
    alphaScaleBiasClamp {bInterpResult=true}
    boneToModify {boneName=dyn_hair_12}
    rotation=<rot 9.886 -38.922 0.146>
    translation=<0.064 -0.806>
   }
   n3 = "Transform (Modify) Bone - Bone: dyn_hair_22"(ComponentPose=@n2, Alpha=1, Translation="0,0,0", Rotation="0.001,-38.045,0.010", Scale="1.000,1.000,0.740") {
    rotationMode=BMM_Additive
    rotationSpace=BCS_BoneSpace
    scaleMode=BMM_Additive
    scaleSpace=BCS_BoneSpace
    translationMode=BMM_Additive
    translationSpace=BCS_BoneSpace
    alphaBoolBlend {blendOption=Linear}
    alphaScaleBiasClamp {bInterpResult=true}
    boneToModify {boneName=dyn_hair_22}
    rotation=<rot 0.001 -38.045 0.01>
    scale {z=0.74}
   }
   n4 = "Transform (Modify) Bone - Bone: dyn_hair_29"(ComponentPose=@n3, Alpha=1, Translation="0,0,0", Rotation="-0.021,-35.200,-0.032", Scale="1.000,1.000,0.740") {
    rotationMode=BMM_Additive
    rotationSpace=BCS_BoneSpace
    scaleMode=BMM_Additive
    scaleSpace=BCS_BoneSpace
    translationMode=BMM_Additive
    translationSpace=BCS_BoneSpace
    alphaBoolBlend {blendOption=Linear}
    alphaScaleBiasClamp {bInterpResult=true}
    boneToModify {boneName=dyn_hair_29}
    rotation=<rot -0.021 -35.1999 -0.032>
    scale {z=0.74}
   }
   Output Pose(Result=Component To Local(ComponentPose=@n4)) {layerGroup=None}
   n5 = Control Rig!(
    Source=Copy Pose From Mesh() {bUseAttachedParent=true}
    Alpha
    CrouchingBool:bool=Get bIsCrouching()
    FloatingBool:bool=Get bIsSkydiveFloating()
    StandingBool:bool=Get Alpha_Standing() -> (Alpha_Standing:bool)
   ) -> (Pose) {
    alphaBoolBlend {blendOption=Linear}
    controlRigAssetReference {blueprintRigClass=&/M_MED_FluteLamp_FaceAcc_CtrlRig.M_MED_FluteLamp_FaceAcc_CtrlRig_C}
   }
   Save cached pose 'ControlRigCache'(Pose=@n5) {cacheName=ControlRigCache}
   n6 = Anim Dynamics - Bone: dyn_hair_01!(ComponentPose=Local To Component(LocalPose=Use cached pose 'ControlRigCache'()), Alpha=1, GravityOverride:Vector="0,0,-980", ExternalForce) -> (Pose) {
    angularDampingOverride=1
    angularSpringConstant=20000
    bAngularSpring=true
    bOverrideAngularDamping=true
    bUseGravityOverride=true
    gravityScale=0.35
    simulationSpace=RootRelative
    alphaBoolBlend {blendOption=Linear}
    boundBone {boneName=dyn_hair_01}
    componentAppliedLinearAccClamp=<3000 3000 3000>
    componentLinearAccScale=<0.85 0.85 0.85>
    componentLinearVelScale=<0.85 0.85 0.85>
    physicsBodyDefinitions [{
      boundBone {boneName=dyn_hair_01}
      boxExtents=<20 40 40>
      constraintSetup {angularLimitsMax=<5 15 5>, angularLimitsMin=<-5 -15 -15>, angularTarget=<1 0 0>}
      localJointOffset=<4 0 0>
     }]
    simSpaceSettings {simSpaceAngularAlpha=0.25}
   }
   n7 = Anim Dynamics - Bone: dyn_hair_02!(ComponentPose=@n6, Alpha=1, GravityOverride:Vector="0,0,-980", ExternalForce) -> (Pose) {
    angularDampingOverride=1
    angularSpringConstant=20000
    bAngularSpring=true
    bOverrideAngularDamping=true
    bUseGravityOverride=true
    gravityScale=0.35
    simulationSpace=RootRelative
    alphaBoolBlend {blendOption=Linear}
    boundBone {boneName=dyn_hair_02}
    componentAppliedLinearAccClamp=<3000 3000 3000>
    componentLinearAccScale=<0.85 0.85 0.85>
    componentLinearVelScale=<0.85 0.85 0.85>
    physicsBodyDefinitions [{
      boundBone {boneName=dyn_hair_02}
      boxExtents=<20 40 40>
      constraintSetup {angularLimitsMax=<5 15 5>, angularLimitsMin=<-5 -15 -15>, angularTarget=<1 0 0>}
      localJointOffset=<4 0 0>
     }]
    simSpaceSettings {simSpaceAngularAlpha=0.25}
   }
   n8 = Anim Dynamics - Bone: dyn_hair_03!(ComponentPose=@n7, Alpha=1, GravityOverride:Vector="0,0,-980", ExternalForce="0,0,0") -> (Pose) {
    angularDampingOverride=1
    angularSpringConstant=20000
    bAngularSpring=true
    bOverrideAngularDamping=true
    bUseGravityOverride=true
    gravityScale=0.35
    simulationSpace=RootRelative
    alphaBoolBlend {blendOption=Linear}
    boundBone {boneName=dyn_hair_03}
    componentAppliedLinearAccClamp=<3000 3000 3000>
    componentLinearAccScale=<0.85 0.85 0.85>
    componentLinearVelScale=<0.85 0.85 0.85>
    physicsBodyDefinitions [{
      boundBone {boneName=dyn_hair_03}
      boxExtents=<20 40 40>
      constraintSetup {angularLimitsMax=<5 10 0>, angularLimitsMin=<-5 -10 -5>, angularTarget=<1 0 0>}
      localJointOffset=<4 0 0>
     }]
    simSpaceSettings {simSpaceAngularAlpha=0.25}
   }
   n9 = Anim Dynamics - Bone: dyn_hair_04!(ComponentPose=@n8, Alpha=1, GravityOverride:Vector="0,0,-980", ExternalForce="0,0,0") -> (Pose) {
    angularDampingOverride=1
    angularSpringConstant=20000
    bAngularSpring=true
    bOverrideAngularDamping=true
    bUseGravityOverride=true
    gravityScale=0.35
    simulationSpace=RootRelative
    alphaBoolBlend {blendOption=Linear}
    boundBone {boneName=dyn_hair_04}
    componentAppliedLinearAccClamp=<3000 3000 3000>
    componentLinearAccScale=<0.85 0.85 0.85>
    componentLinearVelScale=<0.85 0.85 0.85>
    physicsBodyDefinitions [{
      boundBone {boneName=dyn_hair_04}
      boxExtents=<20 40 40>
      constraintSetup {angularLimitsMax=<5 -8 15>, angularLimitsMin=<-5 -7 -15>, angularTarget=<1 0 0>}
      localJointOffset=<4 0 0>
     }]
    simSpaceSettings {simSpaceAngularAlpha=0.25}
   }
   n10 = Anim Dynamics - Bone: dyn_hair_05!(ComponentPose=@n9, Alpha=1, GravityOverride:Vector="0,0,-980", ExternalForce="0,0,0") -> (Pose) {
    angularDampingOverride=1
    angularSpringConstant=20000
    bAngularSpring=true
    bOverrideAngularDamping=true
    bUseGravityOverride=true
    gravityScale=0.35
    simulationSpace=RootRelative
    alphaBoolBlend {blendOption=Linear}
    boundBone {boneName=dyn_hair_05}
    componentAppliedLinearAccClamp=<3000 3000 3000>
    componentLinearAccScale=<0.85 0.85 0.85>
    componentLinearVelScale=<0.85 0.85 0.85>
    physicsBodyDefinitions [{
      boundBone {boneName=dyn_hair_05}
      boxExtents=<20 40 40>
      constraintSetup {angularLimitsMax=<5 15 5>, angularLimitsMin=<-5 -15 -15>, angularTarget=<1 0 0>}
      localJointOffset=<4 0 0>
     }]
    simSpaceSettings {simSpaceAngularAlpha=0.25}
   }
   n11 = Anim Dynamics - Bone: dyn_hair_06!(ComponentPose=@n10, Alpha=1, GravityOverride:Vector="0,0,-980", ExternalForce="0,0,0") -> (Pose) {
    angularDampingOverride=1
    angularSpringConstant=20000
    bAngularSpring=true
    bOverrideAngularDamping=true
    bUseGravityOverride=true
    gravityScale=0.35
    simulationSpace=RootRelative
    alphaBoolBlend {blendOption=Linear}
    boundBone {boneName=dyn_hair_06}
    componentAppliedLinearAccClamp=<3000 3000 3000>
    componentLinearAccScale=<0.85 0.85 0.85>
    componentLinearVelScale=<0.85 0.85 0.85>
    physicsBodyDefinitions [{
      boundBone {boneName=dyn_hair_06}
      boxExtents=<20 40 40>
      constraintSetup {angularLimitsMax=<5 10 5>, angularLimitsMin=<-5 -15 -15>, angularTarget=<1 0 0>}
      localJointOffset=<8 0 0>
     }]
    simSpaceSettings {simSpaceAngularAlpha=0.3}
   }
   n12 = Anim Dynamics - Bone: dyn_hair_07!(ComponentPose=@n11, Alpha=1, GravityOverride:Vector="0,0,-980", ExternalForce="0,0,0") -> (Pose) {
    angularDampingOverride=1
    angularSpringConstant=20000
    bAngularSpring=true
    bOverrideAngularDamping=true
    bUseGravityOverride=true
    gravityScale=0.35
    simulationSpace=RootRelative
    alphaBoolBlend {blendOption=Linear}
    boundBone {boneName=dyn_hair_07}
    componentAppliedLinearAccClamp=<3000 3000 3000>
    componentLinearAccScale=<0.85 0.85 0.85>
    componentLinearVelScale=<0.85 0.85 0.85>
    physicsBodyDefinitions [{
      boundBone {boneName=dyn_hair_07}
      boxExtents=<20 40 40>
      constraintSetup {angularLimitsMax=<5 15 15>, angularLimitsMin=<-5 -15 -15>, angularTarget=<1 0 0>}
      localJointOffset=<6 0 0>
     }]
    simSpaceSettings {simSpaceAngularAlpha=0.3}
   }
   n13 = Anim Dynamics - Bone: dyn_hair_08!(ComponentPose=@n12, Alpha=1, GravityOverride:Vector="0,0,-980", ExternalForce="0,0,0") -> (Pose) {
    angularDampingOverride=1
    angularSpringConstant=20000
    bAngularSpring=true
    bOverrideAngularDamping=true
    bUseGravityOverride=true
    gravityScale=0.35
    simulationSpace=RootRelative
    alphaBoolBlend {blendOption=Linear}
    boundBone {boneName=dyn_hair_08}
    componentAppliedLinearAccClamp=<3000 3000 3000>
    componentLinearAccScale=<0.85 0.85 0.85>
    componentLinearVelScale=<0.85 0.85 0.85>
    physicsBodyDefinitions [{
      boundBone {boneName=dyn_hair_08}
      boxExtents=<20 40 40>
      constraintSetup {angularLimitsMax=<5 5 5>, angularLimitsMin=<-5 -5 -5>, angularTarget=<1 0 0>}
      localJointOffset=<4 0 0>
     }]
    simSpaceSettings {simSpaceAngularAlpha=0.3}
   }
   n14 = Anim Dynamics - Bone: dyn_hair_09!(ComponentPose=@n13, Alpha=1, GravityOverride:Vector="0,0,-980", ExternalForce="0,0,0") -> (Pose) {
    angularDampingOverride=1
    angularSpringConstant=20000
    bAngularSpring=true
    bOverrideAngularDamping=true
    bUseGravityOverride=true
    gravityScale=0.35
    simulationSpace=RootRelative
    alphaBoolBlend {blendOption=Linear}
    boundBone {boneName=dyn_hair_09}
    componentAppliedLinearAccClamp=<3000 3000 3000>
    componentLinearAccScale=<0.85 0.85 0.85>
    componentLinearVelScale=<0.85 0.85 0.85>
    physicsBodyDefinitions [{
      boundBone {boneName=dyn_hair_09}
      boxExtents=<20 40 40>
      constraintSetup {angularLimitsMax=<5 7 8>, angularLimitsMin=<-5 -7 -8>, angularTarget=<1 0 0>}
      localJointOffset=<4 0 0>
     }]
    simSpaceSettings {simSpaceAngularAlpha=0.3}
   }
   n15 = Anim Dynamics - Bone: dyn_hair_10!(ComponentPose=@n14, Alpha=1, GravityOverride:Vector="0,0,-980", ExternalForce="0,0,0") -> (Pose) {
    angularDampingOverride=1
    angularSpringConstant=20000
    bAngularSpring=true
    bOverrideAngularDamping=true
    bUseGravityOverride=true
    gravityScale=0.35
    simulationSpace=RootRelative
    alphaBoolBlend {blendOption=Linear}
    boundBone {boneName=dyn_hair_10}
    componentAppliedLinearAccClamp=<3000 3000 3000>
    componentLinearAccScale=<0.85 0.85 0.85>
    componentLinearVelScale=<0.85 0.85 0.85>
    physicsBodyDefinitions [{
      boundBone {boneName=dyn_hair_10}
      boxExtents=<20 40 40>
      constraintSetup {angularLimitsMax=<5 15 15>, angularLimitsMin=<-5 -10 -12>, angularTarget=<1 0 0>}
      localJointOffset=<7 0 0>
     }]
    simSpaceSettings {simSpaceAngularAlpha=0.3}
   }
   n16 = Anim Dynamics - Bone: dyn_hair_11!(ComponentPose=@n15, Alpha=1, GravityOverride:Vector="0,0,-980", ExternalForce="0,0,0") -> (Pose) {
    angularDampingOverride=1
    angularSpringConstant=20000
    bAngularSpring=true
    bOverrideAngularDamping=true
    bUseGravityOverride=true
    gravityScale=0.35
    simulationSpace=RootRelative
    alphaBoolBlend {blendOption=Linear}
    boundBone {boneName=dyn_hair_11}
    componentAppliedLinearAccClamp=<3000 3000 3000>
    componentLinearAccScale=<0.85 0.85 0.85>
    componentLinearVelScale=<0.85 0.85 0.85>
    physicsBodyDefinitions [{
      boundBone {boneName=dyn_hair_11}
      boxExtents=<20 40 40>
      constraintSetup {angularLimitsMax=<5 10 5>, angularLimitsMin=<-5 -10 -15>, angularTarget=<1 0 0>}
      localJointOffset=<4 0 0>
     }]
    simSpaceSettings {simSpaceAngularAlpha=0.3}
   }
   n17 = Anim Dynamics - Bone: dyn_hair_12!(ComponentPose=@n16, Alpha=1, GravityOverride:Vector="0,0,-980", ExternalForce="0,0,0") -> (Pose) {
    angularDampingOverride=1
    angularSpringConstant=20000
    bAngularSpring=true
    bOverrideAngularDamping=true
    bUseGravityOverride=true
    gravityScale=0.35
    simulationSpace=RootRelative
    alphaBoolBlend {blendOption=Linear}
    boundBone {boneName=dyn_hair_12}
    componentAppliedLinearAccClamp=<3000 3000 3000>
    componentLinearAccScale=<0.85 0.85 0.85>
    componentLinearVelScale=<0.85 0.85 0.85>
    physicsBodyDefinitions [{
      boundBone {boneName=dyn_hair_12}
      boxExtents=<20 40 40>
      constraintSetup {angularLimitsMax=<5 10 5>, angularLimitsMin=<-5 -10 -15>, angularTarget=<1 0 0>}
      localJointOffset=<4 0 0>
     }]
    simSpaceSettings {simSpaceAngularAlpha=0.3}
   }
   n18 = Anim Dynamics - Bone: dyn_hair_13!(ComponentPose=@n17, Alpha=1, GravityOverride:Vector="0,0,-980", ExternalForce="0,0,0") -> (Pose) {
    angularDampingOverride=1
    angularSpringConstant=20000
    bAngularSpring=true
    bOverrideAngularDamping=true
    bUseGravityOverride=true
    gravityScale=0.35
    simulationSpace=RootRelative
    alphaBoolBlend {blendOption=Linear}
    boundBone {boneName=dyn_hair_13}
    componentAppliedLinearAccClamp=<3000 3000 3000>
    componentLinearAccScale=<0.6 0.6 0.6>
    componentLinearVelScale=<0.6 0.6 0.6>
    physicsBodyDefinitions [{
      boundBone {boneName=dyn_hair_13}
      boxExtents=<20 40 40>
      constraintSetup {angularLimitsMax=<5 15 0>, angularLimitsMin=<-5 -15 -15>, angularTarget=<1 0 0>}
      localJointOffset=<3 0 0>
     }]
    simSpaceSettings {simSpaceAngularAlpha=0.3}
   }
   n19 = Anim Dynamics - Bone: dyn_hair_14!(ComponentPose=@n18, Alpha=1, GravityOverride:Vector="0,0,-980", ExternalForce="0,0,0") -> (Pose) {
    angularDampingOverride=1
    angularSpringConstant=20000
    bAngularSpring=true
    bOverrideAngularDamping=true
    bUseGravityOverride=true
    gravityScale=0.35
    simulationSpace=RootRelative
    alphaBoolBlend {blendOption=Linear}
    boundBone {boneName=dyn_hair_14}
    componentAppliedLinearAccClamp=<3000 3000 3000>
    componentLinearAccScale=<0.6 0.6 0.6>
    componentLinearVelScale=<0.6 0.6 0.6>
    physicsBodyDefinitions [{
      boundBone {boneName=dyn_hair_14}
      boxExtents=<20 40 40>
      constraintSetup {angularLimitsMax=<5 15 0>, angularLimitsMin=<-5 -15 -5>, angularTarget=<1 0 0>}
      localJointOffset=<3 0 0>
     }]
    simSpaceSettings {simSpaceAngularAlpha=0.3}
   }
   n20 = Anim Dynamics - Bone: dyn_hair_15!(ComponentPose=@n19, Alpha=1, GravityOverride:Vector="0,0,-980", ExternalForce="0,0,0") -> (Pose) {
    angularDampingOverride=1
    angularSpringConstant=20000
    bAngularSpring=true
    bOverrideAngularDamping=true
    bUseGravityOverride=true
    gravityScale=0.35
    simulationSpace=RootRelative
    alphaBoolBlend {blendOption=Linear}
    boundBone {boneName=dyn_hair_15}
    componentAppliedLinearAccClamp=<3000 3000 3000>
    componentLinearAccScale=<0.85 0.85 0.85>
    componentLinearVelScale=<0.85 0.85 0.85>
    physicsBodyDefinitions [{
      boundBone {boneName=dyn_hair_15}
      boxExtents=<20 40 40>
      constraintSetup {angularLimitsMax=<5 15 15>, angularLimitsMin=<-5 -15 -15>, angularTarget=<1 0 0>}
      localJointOffset=<10 0 0>
     }]
    simSpaceSettings {simSpaceAngularAlpha=0.3}
   }
   n21 = Anim Dynamics - Bone: dyn_hair_16!(ComponentPose=@n20, Alpha=1, GravityOverride:Vector="0,0,-980", ExternalForce="0,0,0") -> (Pose) {
    angularDampingOverride=1
    angularSpringConstant=20000
    bAngularSpring=true
    bOverrideAngularDamping=true
    bUseGravityOverride=true
    gravityScale=0.35
    simulationSpace=RootRelative
    alphaBoolBlend {blendOption=Linear}
    boundBone {boneName=dyn_hair_16}
    componentAppliedLinearAccClamp=<3000 3000 3000>
    componentLinearAccScale=<0.85 0.85 0.85>
    componentLinearVelScale=<0.85 0.85 0.85>
    physicsBodyDefinitions [{
      boundBone {boneName=dyn_hair_16}
      boxExtents=<20 40 40>
      constraintSetup {angularLimitsMax=<5 15 15>, angularLimitsMin=<-5 -15 -15>, angularTarget=<1 0 0>}
      localJointOffset=<10 0 0>
     }]
    simSpaceSettings {simSpaceAngularAlpha=0.3}
   }
   n22 = Anim Dynamics - Bone: dyn_hair_17!(ComponentPose=@n21, Alpha=1, GravityOverride:Vector="0,0,-980", ExternalForce="0,0,0") -> (Pose) {
    angularDampingOverride=1
    angularSpringConstant=20000
    bAngularSpring=true
    bOverrideAngularDamping=true
    bUseGravityOverride=true
    gravityScale=0.35
    simulationSpace=RootRelative
    alphaBoolBlend {blendOption=Linear}
    boundBone {boneName=dyn_hair_17}
    componentAppliedLinearAccClamp=<3000 3000 3000>
    componentLinearAccScale=<0.85 0.85 0.85>
    componentLinearVelScale=<0.85 0.85 0.85>
    physicsBodyDefinitions [{
      boundBone {boneName=dyn_hair_17}
      boxExtents=<20 40 40>
      constraintSetup {angularLimitsMax=<5 15 15>, angularLimitsMin=<-5 -15 -15>, angularTarget=<1 0 0>}
      localJointOffset=<4 0 0>
     }]
    simSpaceSettings {simSpaceAngularAlpha=0.3}
   }
   n23 = Anim Dynamics - Bone: dyn_hair_18!(ComponentPose=@n22, Alpha=1, GravityOverride:Vector="0,0,-980", ExternalForce="0,0,0") -> (Pose) {
    angularDampingOverride=1
    angularSpringConstant=20000
    bAngularSpring=true
    bOverrideAngularDamping=true
    bUseGravityOverride=true
    gravityScale=0.35
    simulationSpace=RootRelative
    alphaBoolBlend {blendOption=Linear}
    boundBone {boneName=dyn_hair_18}
    componentAppliedLinearAccClamp=<3000 3000 3000>
    componentLinearAccScale=<0.85 0.85 0.85>
    componentLinearVelScale=<0.85 0.85 0.85>
    physicsBodyDefinitions [{
      boundBone {boneName=dyn_hair_18}
      boxExtents=<20 40 40>
      constraintSetup {angularLimitsMax=<5 15 15>, angularLimitsMin=<-5 -15 -15>, angularTarget=<1 0 0>}
      localJointOffset=<10 0 0>
     }]
    simSpaceSettings {simSpaceAngularAlpha=0.3}
   }
   n24 = Anim Dynamics - Bone: dyn_hair_19!(ComponentPose=@n23, Alpha=1, GravityOverride:Vector="0,0,-980", ExternalForce="0,0,0") -> (Pose) {
    angularDampingOverride=1
    angularSpringConstant=20000
    bAngularSpring=true
    bOverrideAngularDamping=true
    bUseGravityOverride=true
    gravityScale=0.35
    simulationSpace=RootRelative
    alphaBoolBlend {blendOption=Linear}
    boundBone {boneName=dyn_hair_19}
    componentAppliedLinearAccClamp=<3000 3000 3000>
    componentLinearAccScale=<0.85 0.85 0.85>
    componentLinearVelScale=<0.85 0.85 0.85>
    physicsBodyDefinitions [{
      boundBone {boneName=dyn_hair_19}
      boxExtents=<20 40 40>
      constraintSetup {angularLimitsMax=<5 15 15>, angularLimitsMin=<-5 -15 -15>, angularTarget=<1 0 0>}
      localJointOffset=<4 0 0>
     }]
    simSpaceSettings {simSpaceAngularAlpha=0.3}
   }
   n25 = Anim Dynamics - Bone: dyn_hair_20!(ComponentPose=@n24, Alpha=1, GravityOverride:Vector="0,0,-980", ExternalForce="0,0,0") -> (Pose) {
    angularDampingOverride=1
    angularSpringConstant=20000
    bAngularSpring=true
    bOverrideAngularDamping=true
    bUseGravityOverride=true
    gravityScale=0.35
    simulationSpace=RootRelative
    alphaBoolBlend {blendOption=Linear}
    boundBone {boneName=dyn_hair_20}
    componentAppliedLinearAccClamp=<3000 3000 3000>
    componentLinearAccScale=<0.85 0.85 0.85>
    componentLinearVelScale=<0.85 0.85 0.85>
    physicsBodyDefinitions [{
      boundBone {boneName=dyn_hair_20}
      boxExtents=<20 40 40>
      constraintSetup {angularLimitsMax=<5 15 15>, angularLimitsMin=<-5 -15 -15>, angularTarget=<1 0 0>}
      localJointOffset=<10 0 0>
     }]
    simSpaceSettings {simSpaceAngularAlpha=0.3}
   }
   n26 = Anim Dynamics - Bone: dyn_hair_21!(ComponentPose=@n25, Alpha=1, GravityOverride:Vector="0,0,-980", ExternalForce="0,0,0") -> (Pose) {
    angularDampingOverride=1
    angularSpringConstant=20000
    bAngularSpring=true
    bOverrideAngularDamping=true
    bUseGravityOverride=true
    gravityScale=0.35
    simulationSpace=RootRelative
    alphaBoolBlend {blendOption=Linear}
    boundBone {boneName=dyn_hair_21}
    componentAppliedLinearAccClamp=<3000 3000 3000>
    componentLinearAccScale=<0.85 0.85 0.85>
    componentLinearVelScale=<0.85 0.85 0.85>
    physicsBodyDefinitions [{
      boundBone {boneName=dyn_hair_21}
      boxExtents=<20 40 40>
      constraintSetup {angularLimitsMax=<5 15 15>, angularLimitsMin=<-5 -15 -15>, angularTarget=<1 0 0>}
      localJointOffset=<11 0 0>
     }]
    simSpaceSettings {simSpaceAngularAlpha=0.3}
   }
   n27 = Anim Dynamics - Bone: dyn_hair_22!(ComponentPose=@n26, Alpha=1, GravityOverride:Vector="0,0,-980", ExternalForce="0,0,0") -> (Pose) {
    angularDampingOverride=1
    angularSpringConstant=20000
    bAngularSpring=true
    bOverrideAngularDamping=true
    bUseGravityOverride=true
    gravityScale=0.35
    simulationSpace=RootRelative
    alphaBoolBlend {blendOption=Linear}
    boundBone {boneName=dyn_hair_22}
    componentAppliedLinearAccClamp=<3000 3000 3000>
    componentLinearAccScale=<0.85 0.85 0.85>
    componentLinearVelScale=<0.85 0.85 0.85>
    physicsBodyDefinitions [{
      boundBone {boneName=dyn_hair_22}
      boxExtents=<20 40 40>
      constraintSetup {angularLimitsMax=<5 15 5>, angularLimitsMin=<-5 -15 -15>, angularTarget=<1 0 0>}
      localJointOffset=<4 0 0>
     }]
    simSpaceSettings {simSpaceAngularAlpha=0.3}
   }
   n28 = Anim Dynamics - Bone: dyn_hair_23!(ComponentPose=@n27, Alpha=1, GravityOverride:Vector="0,0,-980", ExternalForce="0,0,0") -> (Pose) {
    angularDampingOverride=1
    angularSpringConstant=20000
    bAngularSpring=true
    bOverrideAngularDamping=true
    bUseGravityOverride=true
    gravityScale=0.35
    simulationSpace=RootRelative
    alphaBoolBlend {blendOption=Linear}
    boundBone {boneName=dyn_hair_23}
    componentAppliedLinearAccClamp=<3000 3000 3000>
    componentLinearAccScale=<0.85 0.85 0.85>
    componentLinearVelScale=<0.85 0.85 0.85>
    physicsBodyDefinitions [{
      boundBone {boneName=dyn_hair_23}
      boxExtents=<20 40 40>
      constraintSetup {angularLimitsMax=<5 15 15>, angularLimitsMin=<-5 -15 -15>, angularTarget=<1 0 0>}
      localJointOffset=<11 0 0>
     }]
    simSpaceSettings {simSpaceAngularAlpha=0.3}
   }
   n29 = Anim Dynamics - Bone: dyn_hair_24!(ComponentPose=@n28, Alpha=1, GravityOverride:Vector="0,0,-980", ExternalForce="0,0,0") -> (Pose) {
    angularDampingOverride=1
    angularSpringConstant=20000
    bAngularSpring=true
    bOverrideAngularDamping=true
    bUseGravityOverride=true
    gravityScale=0.35
    simulationSpace=RootRelative
    alphaBoolBlend {blendOption=Linear}
    boundBone {boneName=dyn_hair_24}
    componentAppliedLinearAccClamp=<3000 3000 3000>
    componentLinearAccScale=<0.85 0.85 0.85>
    componentLinearVelScale=<0.85 0.85 0.85>
    physicsBodyDefinitions [{
      boundBone {boneName=dyn_hair_24}
      boxExtents=<20 40 40>
      constraintSetup {angularLimitsMax=<5 15 15>, angularLimitsMin=<-5 -15 -15>, angularTarget=<1 0 0>}
      localJointOffset=<4 0 0>
     }]
    simSpaceSettings {simSpaceAngularAlpha=0.3}
   }
   n30 = Anim Dynamics - Bone: dyn_hair_b_01!(ComponentPose=@n29, Alpha=1, GravityOverride:Vector="0,0,-980", ExternalForce="0,0,0") -> (Pose) {
    angularDampingOverride=1
    angularSpringConstant=20000
    bAngularSpring=true
    bOverrideAngularDamping=true
    bUseGravityOverride=true
    gravityScale=0.35
    simulationSpace=RootRelative
    alphaBoolBlend {blendOption=Linear}
    boundBone {boneName=dyn_hair_b_01}
    componentAppliedLinearAccClamp=<3000 3000 3000>
    componentLinearAccScale=<0.85 0.85 0.85>
    componentLinearVelScale=<0.85 0.85 0.85>
    physicsBodyDefinitions [{
      boundBone {boneName=dyn_hair_b_01}
      boxExtents=<20 40 40>
      constraintSetup {angularLimitsMax=<5 15 15>, angularLimitsMin=<-5 -15 -15>, angularTarget=<1 0 0>}
      localJointOffset=<11 0 0>
     }]
    simSpaceSettings {simSpaceAngularAlpha=0.3}
   }
   n31 = Anim Dynamics - Bone: dyn_hair_25!(ComponentPose=@n30, Alpha=1, GravityOverride:Vector="0,0,-980", ExternalForce="0,0,0") -> (Pose) {
    angularDampingOverride=1
    angularSpringConstant=20000
    bAngularSpring=true
    bOverrideAngularDamping=true
    bUseGravityOverride=true
    gravityScale=0.35
    simulationSpace=RootRelative
    alphaBoolBlend {blendOption=Linear}
    boundBone {boneName=dyn_hair_25}
    componentAppliedLinearAccClamp=<3000 3000 3000>
    componentLinearAccScale=<0.85 0.85 0.85>
    componentLinearVelScale=<0.85 0.85 0.85>
    physicsBodyDefinitions [{
      boundBone {boneName=dyn_hair_25}
      boxExtents=<20 40 40>
      constraintSetup {angularLimitsMax=<5 15 15>, angularLimitsMin=<-5 -15 -15>, angularTarget=<1 0 0>}
      localJointOffset=<4 0 0>
     }]
    simSpaceSettings {simSpaceAngularAlpha=0.3}
   }
   n32 = Anim Dynamics - Bone: dyn_hair_b_02!(ComponentPose=@n31, Alpha=1, GravityOverride:Vector="0,0,-980", ExternalForce="0,0,0") -> (Pose) {
    angularDampingOverride=1
    angularSpringConstant=20000
    bAngularSpring=true
    bOverrideAngularDamping=true
    bUseGravityOverride=true
    gravityScale=0.35
    simulationSpace=RootRelative
    alphaBoolBlend {blendOption=Linear}
    boundBone {boneName=dyn_hair_b_02}
    componentAppliedLinearAccClamp=<3000 3000 3000>
    componentLinearAccScale=<0.85 0.85 0.85>
    componentLinearVelScale=<0.85 0.85 0.85>
    physicsBodyDefinitions [{
      boundBone {boneName=dyn_hair_b_02}
      boxExtents=<20 40 40>
      constraintSetup {angularLimitsMax=<5 15 15>, angularLimitsMin=<-5 -15 -15>, angularTarget=<1 0 0>}
      localJointOffset=<4 0 0>
     }]
    simSpaceSettings {simSpaceAngularAlpha=0.3}
   }
   n33 = Anim Dynamics - Bone: dyn_hair_26!(ComponentPose=@n32, Alpha=1, GravityOverride:Vector="0,0,-980", ExternalForce="0,0,0") -> (Pose) {
    angularDampingOverride=1
    angularSpringConstant=20000
    bAngularSpring=true
    bOverrideAngularDamping=true
    bUseGravityOverride=true
    gravityScale=0.35
    simulationSpace=RootRelative
    alphaBoolBlend {blendOption=Linear}
    boundBone {boneName=dyn_hair_26}
    componentAppliedLinearAccClamp=<3000 3000 3000>
    componentLinearAccScale=<0.85 0.85 0.85>
    componentLinearVelScale=<0.85 0.85 0.85>
    physicsBodyDefinitions [{
      boundBone {boneName=dyn_hair_26}
      boxExtents=<20 30 30>
      constraintSetup {angularLimitsMax=<5 15 15>, angularLimitsMin=<-5 -15 -15>, angularTarget=<1 0 0>}
      localJointOffset=<12 0 0>
     }]
    simSpaceSettings {simSpaceAngularAlpha=0.3}
   }
   n34 = Anim Dynamics - Bone: dyn_hair_27!(ComponentPose=@n33, Alpha=1, GravityOverride:Vector="0,0,-980", ExternalForce="0,0,0") -> (Pose) {
    angularDampingOverride=1
    angularSpringConstant=20000
    bAngularSpring=true
    bOverrideAngularDamping=true
    bUseGravityOverride=true
    gravityScale=0.35
    simulationSpace=RootRelative
    alphaBoolBlend {blendOption=Linear}
    boundBone {boneName=dyn_hair_27}
    componentAppliedLinearAccClamp=<3000 3000 3000>
    componentLinearAccScale=<0.85 0.85 0.85>
    componentLinearVelScale=<0.85 0.85 0.85>
    physicsBodyDefinitions [{
      boundBone {boneName=dyn_hair_27}
      boxExtents=<20 30 30>
      constraintSetup {angularLimitsMax=<5 15 15>, angularLimitsMin=<-5 -15 -15>, angularTarget=<1 0 0>}
      localJointOffset=<4 0 0>
     }]
    simSpaceSettings {simSpaceAngularAlpha=0.3}
   }
   n35 = Anim Dynamics - Bone: dyn_hair_28!(ComponentPose=@n34, Alpha=1, GravityOverride:Vector="0,0,-980", ExternalForce="0,0,0") -> (Pose) {
    angularDampingOverride=1
    angularSpringConstant=20000
    bAngularSpring=true
    bOverrideAngularDamping=true
    bUseGravityOverride=true
    gravityScale=0.35
    simulationSpace=RootRelative
    alphaBoolBlend {blendOption=Linear}
    boundBone {boneName=dyn_hair_28}
    componentAppliedLinearAccClamp=<3000 3000 3000>
    componentLinearAccScale=<0.85 0.85 0.85>
    componentLinearVelScale=<0.85 0.85 0.85>
    physicsBodyDefinitions [{
      boundBone {boneName=dyn_hair_28}
      boxExtents=<20 30 30>
      constraintSetup {angularLimitsMax=<5 15 15>, angularLimitsMin=<-5 -15 -15>, angularTarget=<1 0 0>}
      localJointOffset=<12 0 0>
     }]
    simSpaceSettings {simSpaceAngularAlpha=0.3}
   }
   n36 = Anim Dynamics - Bone: dyn_hair_29!(ComponentPose=@n35, Alpha=1, GravityOverride:Vector="0,0,-980", ExternalForce="0,0,0") -> (Pose) {
    angularDampingOverride=1
    angularSpringConstant=20000
    bAngularSpring=true
    bOverrideAngularDamping=true
    bUseGravityOverride=true
    gravityScale=0.35
    simulationSpace=RootRelative
    alphaBoolBlend {blendOption=Linear}
    boundBone {boneName=dyn_hair_29}
    componentAppliedLinearAccClamp=<3000 3000 3000>
    componentLinearAccScale=<0.85 0.85 0.85>
    componentLinearVelScale=<0.85 0.85 0.85>
    physicsBodyDefinitions [{
      boundBone {boneName=dyn_hair_29}
      boxExtents=<20 30 30>
      constraintSetup {angularLimitsMax=<5 15 15>, angularLimitsMin=<-5 -15 -15>, angularTarget=<1 0 0>}
      localJointOffset=<4 0 0>
     }]
    simSpaceSettings {simSpaceAngularAlpha=0.3}
   }
   n37 = Anim Dynamics - Bone: dyn_hair_30!(ComponentPose=@n36, Alpha=1, GravityOverride:Vector="0,0,-980", ExternalForce="0,0,0") -> (Pose) {
    angularDampingOverride=1
    angularSpringConstant=20000
    bAngularSpring=true
    bOverrideAngularDamping=true
    bUseGravityOverride=true
    gravityScale=0.35
    simulationSpace=RootRelative
    alphaBoolBlend {blendOption=Linear}
    boundBone {boneName=dyn_hair_30}
    componentAppliedLinearAccClamp=<3000 3000 3000>
    componentLinearAccScale=<0.85 0.85 0.85>
    componentLinearVelScale=<0.85 0.85 0.85>
    physicsBodyDefinitions [{
      boundBone {boneName=dyn_hair_30}
      boxExtents=<20 30 30>
      constraintSetup {angularLimitsMax=<5 15 15>, angularLimitsMin=<-5 -15 -15>, angularTarget=<1 0 0>}
      localJointOffset=<12 0 0>
     }]
    simSpaceSettings {simSpaceAngularAlpha=0.3}
   }
   n38 = Anim Dynamics - Bone: dyn_hair_31!(ComponentPose=@n37, Alpha=1, GravityOverride:Vector="0,0,-980", ExternalForce="0,0,0") -> (Pose) {
    angularDampingOverride=1
    angularSpringConstant=20000
    bAngularSpring=true
    bOverrideAngularDamping=true
    bUseGravityOverride=true
    gravityScale=0.35
    simulationSpace=RootRelative
    alphaBoolBlend {blendOption=Linear}
    boundBone {boneName=dyn_hair_31}
    componentAppliedLinearAccClamp=<3000 3000 3000>
    componentLinearAccScale=<0.85 0.85 0.85>
    componentLinearVelScale=<0.85 0.85 0.85>
    physicsBodyDefinitions [{
      boundBone {boneName=dyn_hair_31}
      boxExtents=<20 30 30>
      constraintSetup {angularLimitsMax=<5 15 15>, angularLimitsMin=<-5 -15 -15>, angularTarget=<1 0 0>}
      localJointOffset=<4 0 0>
     }]
    simSpaceSettings {simSpaceAngularAlpha=0.3}
   }
   n39 = Anim Dynamics - Bone: dyn_hair_32!(ComponentPose=@n38, Alpha=1, GravityOverride:Vector="0,0,-980", ExternalForce="0,0,0") -> (Pose) {
    angularDampingOverride=1
    angularSpringConstant=20000
    bAngularSpring=true
    bOverrideAngularDamping=true
    bUseGravityOverride=true
    gravityScale=0.35
    simulationSpace=RootRelative
    alphaBoolBlend {blendOption=Linear}
    boundBone {boneName=dyn_hair_32}
    componentAppliedLinearAccClamp=<3000 3000 3000>
    componentLinearAccScale=<0.85 0.85 0.85>
    componentLinearVelScale=<0.85 0.85 0.85>
    physicsBodyDefinitions [{
      boundBone {boneName=dyn_hair_32}
      boxExtents=<20 30 30>
      constraintSetup {angularLimitsMax=<5 15 15>, angularLimitsMin=<-5 -15 -15>, angularTarget=<1 0 0>}
      localJointOffset=<12 0 0>
     }]
    simSpaceSettings {simSpaceAngularAlpha=0.3}
   }
   n40 = Anim Dynamics - Bone: dyn_hair_33(ComponentPose=@n39, Alpha=1, ExternalForce="0,0,0") {
    angularDampingOverride=1
    angularSpringConstant=20000
    bAngularSpring=true
    bOverrideAngularDamping=true
    bUseGravityOverride=true
    gravityScale=0.35
    simulationSpace=RootRelative
    alphaBoolBlend {blendOption=Linear}
    boundBone {boneName=dyn_hair_33}
    componentLinearAccScale=<0.7 0.7 0.7>
    componentLinearVelScale=<0.7 0.7 0.7>
    gravityOverride {z=-980}
    physicsBodyDefinitions [{
      boundBone {boneName=dyn_hair_33}
      boxExtents=<20 30 30>
      constraintSetup {angularLimitsMax=<15 15 0>, angularLimitsMin=<-15 -15 -15>, angularTarget=<1 0 0>}
      localJointOffset=<4 0 0>
     }]
    simSpaceSettings {simSpaceAngularAlpha=0.2}
   }
   Save cached pose 'DynCache'(Pose=Component To Local(ComponentPose=@n40)) {cacheName=DynCache}
  ]}
 EventGraph {nodes [
   Event BlueprintUpdateAnimation()
   > n1 = Sequence!(execute) -> (then_0, then_1, then_2:Exec)
   > Set Alpha_Standing(Alpha_Standing:bool=NOT Boolean(A=Get bIsCrouching())) -> (Output_Get:bool)
   n2 = Event OnCharacterCustomizationLoaded()
   > n3 = Set Skel Mesh(Skel Mesh=@n2.CustomBaseMesh)
   Set Alpha_HeadRotation(
    execute=@n1.then_2
    Alpha_HeadRotation:float=GetAlphaFromReferenceTransformDelta(SkeletalMeshComponent=@n3, BoneName=neck_02, BaseName=head, InputRange=(X=0,Y=50), TwistAxis="1,0,0")
   ) -> (Output_Get:float)
  ]}
]
variables [
 Alpha_HeadRotation {type=number}
 Alpha_Standing {type=boolean, default=true}
 Gravity Override {type=Vector}
 RBAN Settings Out {type=FortRigidBodyAnimNodeParameters}
 Skel Mesh {type=SkeletalMeshComponent}
]
